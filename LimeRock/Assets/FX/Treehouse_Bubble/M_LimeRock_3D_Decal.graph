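# Translucent deferred decal material for the LimeRock treehouse bubble FX, built around animated noise.
# Two panned noise samples are averaged, raised to Power, multiplied by a U gradient whose exponent is decal color alpha, then tinted by decal RGB.
# A third panned noise scaled by Distortion_Amount ends in an unused Add.
# No material output connections appear, and the OverallOpacity and red Emissive_Color parameters feed nothing.
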Material /LimeRock/Assets/FX/Treehouse_Bubble/M_LimeRock_3D_Decal
graph {
 properties {bAutomaticallySetUsageInEditor=false, blendMode=BLEND_Translucent, materialDomain=MD_DeferredDecal, usedShadingModels=Default Lit}
 nodes [
  n1 = TextureCoordinate()
  n2 = AppendVector(
   A=ScalarParameter() {defaultValue=1, group=Distortion, parameterName=Distortion_Freq_U_01}
   B=ScalarParameter() {defaultValue=1, group=Distortion, parameterName=Distortion_Freq_V_01}
  )
  n3 = ScalarParameter() {group=Distortion, parameterName=Distortion_Speed_U_01}
  n4 = ScalarParameter() {group=Distortion, parameterName=Distortion_Speed_V_01}
  n5 = TextureSampleParameter2D(UVs=Panner(Coordinate=Multiply(A=@n1, B=@n2), Speed=AppendVector(A=@n3, B=@n4)) {speedY=0.5}) {
   group=Distortion
   parameterName=Distortion_Texture
   samplerType=SAMPLERTYPE_LinearColor
   texture=&T_NoisePack_LinearColor
  }
  Add(B=Multiply(A=@n5.G, B=ScalarParameter() {group=Distortion, parameterName=Distortion_Amount}))
  n6 = TextureCoordinate()
  n7 = AppendVector(
   A=ScalarParameter() {defaultValue=2, group=Opacity, parameterName=Opacity_U_Freq_01}
   B=ScalarParameter() {defaultValue=0.5, group=Opacity, parameterName=Opacity_V_Freq_01}
  )
  n8 = ScalarParameter() {group=Opacity, parameterName=Freq_U_01_Speed}
  n9 = ScalarParameter() {defaultValue=0.5, group=Opacity, parameterName=Freq_V_01_Speed}
  n10 = TextureSampleParameter2D(UVs=Panner(Coordinate=Multiply(A=@n6, B=@n7), Speed=AppendVector(A=@n8, B=@n9)) {speedY=0.5}) {
   group=Opacity
   parameterName=Opacity_Texture_01
   samplerType=SAMPLERTYPE_LinearColor
   texture=&T_NoisePack_LinearColor
  }
  n11 = TextureCoordinate()
  n12 = AppendVector(
   A=ScalarParameter() {defaultValue=2, group=Opacity, parameterName=Opacity_U_Freq_02}
   B=ScalarParameter() {defaultValue=0.5, group=Opacity, parameterName=Opacity_V_Freq_02}
  )
  n13 = ScalarParameter() {group=Opacity, parameterName=Freq_U_02_Speed}
  n14 = ScalarParameter() {defaultValue=0.5, group=Opacity, parameterName=Freq_V_02_Speed}
  n15 = TextureSampleParameter2D(UVs=Panner(Coordinate=Multiply(A=@n11, B=@n12), Speed=AppendVector(A=@n13, B=@n14)) {speedY=0.5}) {
   group=Opacity
   parameterName=Opacity_Texture_02
   samplerType=SAMPLERTYPE_LinearColor
   texture=&T_NoisePack_LinearColor
  }
  n16 = Divide(A=Add(A=@n10.G, B=@n15.G))
  n17 = ScalarParameter() {defaultValue=2, parameterName=Power}
  n18 = ComponentMask(None=TextureCoordinate()) {r=true}
  n19 = DecalColor()
  Multiply(A=Multiply(A=Power(Base=@n16, Exp=@n17), B=Power(Base=@n18, Exp=@n19.A)), B=@n19.RGB)
  ScalarParameter() {defaultValue=0.4, parameterName=OverallOpacity}
  VectorParameter() {group=Emissive, parameterName=Emissive_Color, defaultValue=<rgba 1 0 0 0>}
 ]
}
