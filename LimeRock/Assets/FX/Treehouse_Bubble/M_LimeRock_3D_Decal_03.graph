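# Translucent deferred decal material for the LimeRock treehouse bubble FX, drawing panning, contrast-boosted streaks in a two-color emissive blend driven by decal color.
# Panning noise offsets centered UVs before a polar-style remap, although both arctangent inputs read the G component.
# The streak texture samples with derivatives from unwarped UVs.
# Streaks multiply with a sphere mask, a flippable vertical gradient and decal alpha to form the mask.
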Material /LimeRock/Assets/FX/Treehouse_Bubble/M_LimeRock_3D_Decal_03
graph {
 properties {
  bAutomaticallySetUsageInEditor=false
  bUsedWithNiagaraSprites=true
  blendMode=BLEND_Translucent
  materialDomain=MD_DeferredDecal
  usedShadingModels=Default Lit
 }
 nodes [
  n1 = DecalColor()
  MaterialFunctionCall(Emissive=LinearInterpolate(
    A=VectorParameter() {parameterName=Color_One, defaultValue=<rgba 6.7235 1.3982 10 1>}.RGB
    B=VectorParameter() {parameterName=Color_Two, defaultValue=<rgba 0.1269 2.4129 10 1>}.RGB
    Alpha=@n1.RGB
   )) {materialFunction=&MF_FX_EmissiveAdjustments}
  n2 = TextureCoordinate()
  n3 = AppendVector(A=ScalarParameter() {parameterName=Noise_U_Speed}, B=ScalarParameter() {parameterName=Noise_V_Speed})
  n4 = ComponentMask(None=TextureSampleParameter2D(UVs=Panner(Coordinate=@n2, Speed=@n3)) {parameterName=Noise_Texture, texture=&/T_LimeRock_Noise34}.RGB) {g=true, r=true}
  n5 = ScalarParameter() {parameterName=Noise}
  n6 = Subtract(A=Add(A=Multiply(A=@n4, B=@n5), B=TextureCoordinate())) {constB=0.5}
  n7 = Multiply(A=@n6, B=@n6)
  n8 = Add(A=ComponentMask(None=@n7) {r=true}, B=ComponentMask(None=@n7) {g=true})
  n9 = MaterialFunctionCall(Float2=@n6) {materialFunction=&BreakOutFloat2Components}
  n10 = Constant() {r=2}
  n11 = Constant() {r=3.142}
  n12 = Divide(A=AppendVector(A=SquareRoot(None=@n8), B=Arctangent2!(Y=@n9.G, X=@n9.G)), B=AppendVector(A=Constant() {r=1}, B=Multiply(A=@n10, B=@n11)))
  n13 = AppendVector(A=ScalarParameter() {parameterName=U_Speed}, B=ScalarParameter() {parameterName=V_Speed})
  n14 = ScalarParameter() {defaultValue=1, parameterName=U_Freq}
  n15 = ScalarParameter() {defaultValue=1, parameterName=V_Freq}
  n16 = TextureCoordinate()
  n17 = TextureSampleParameter2D!(
   UVs=Multiply(A=Panner(Coordinate=@n12, Speed=@n13), B=AppendVector(A=@n14, B=@n15))
   "DDX(UVs)"=DDX(Value=@n16)
   "DDY(UVs)"=DDY(Value=@n16)
   Apply View MipBias
  ) {
   mipValueMode=TMVM_Derivative
   parameterName=Texture
   samplerType=SAMPLERTYPE_LinearColor
   texture=/Game/TechArt/Seasonal/LackLuster/Textures/T_Lackluster_Streaks
  }
  n18 = Power(Base=@n17.G, Exp=ScalarParameter() {defaultValue=4, parameterName=Contrast})
  n19 = SphereMask(
   A=TextureCoordinate()
   B=Constant() {r=0.5}
   Radius=ScalarParameter() {defaultValue=0.5, parameterName=Mask_Radius}
   Hardness=ScalarParameter() {parameterName=Mask_Hardness}
  )
  n20 = ComponentMask(None=TextureCoordinate()) {g=true}
  n21 = StaticSwitchParameter(True=OneMinus(None=@n20), False=@n20) {defaultValue=true, parameterName=G_Mask_Revert}
  n22 = ScalarParameter() {defaultValue=2, parameterName=G_Power}
  Multiply(A=@n1.A, B=Multiply(A=Multiply(A=@n18, B=@n19), B=Power(Base=@n21, Exp=@n22)))
 ]
}
