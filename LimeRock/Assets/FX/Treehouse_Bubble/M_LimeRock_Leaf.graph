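# Translucent, two-sided, default lit material for leaf particles on Niagara sprites, kept with the LimeRock Treehouse Bubble effects.
# A leaf SubUV texture parameter is tinted by particle color and scaled by an Emissive_Amount scalar.
# The texture alpha is multiplied by a CameraDepthFade with parameterized length and offset, fading leaves by their depth from the camera.
# Particle alpha is not used.
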Material /LimeRock/Assets/FX/Treehouse_Bubble/M_LimeRock_Leaf
graph {
 properties {
  bAutomaticallySetUsageInEditor=false
  bUsedWithNiagaraSprites=true
  blendMode=BLEND_Translucent
  twoSided=true
  usedShadingModels=Default Lit
 }
 nodes [
  n1 = TextureSampleParameter2D() {parameterName=Texture, texture=/Game/Effects/Fort_Effects/Textures/Leaf/T_Leaf_SubUV}
  Multiply(A=Multiply(A=ParticleColor().RGB, B=@n1.RGB), B=ScalarParameter() {parameterName=Emissive_Amount})
  Multiply(
   A=@n1.A
   B=MaterialFunctionCall(Fade Length=ScalarParameter() {parameterName=Fade_length}, Fade Offset=ScalarParameter() {parameterName=Fade_Offset}) {materialFunction=&CameraDepthFade}.Result
  )
 ]
}
