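# Translucent unlit, two-sided material for an energy spline effect on Niagara ribbons and mesh particles, filed under the LimeRock Treehouse Bubble FX.
# Panning wispy noise tinted by a color gradient adds to a scrolling rainbow texture and is scaled by particle color.
# Opacity math combines noise, sine edge masks, a smoke mask, depth fade (deferred path), view-angle and camera-distance fades.
# World position offset math combines SplineThicken, using particle-driven width and camera-proximity collapse, with axis rotation.
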Material /LimeRock/Assets/FX/Treehouse_Bubble/M_LimeRock_Spline_Energy_Suppressed
graph {
 properties {
  bUsedWithNiagaraMeshParticles=true
  bUsedWithNiagaraRibbons=true
  blendMode=BLEND_Translucent
  pixelDepthOffsetMode=PDOM_Legacy
  shadingModel=MSM_Unlit
  twoSided=true
  usedShadingModels=Unlit
 }
 nodes [
  n1 = ActorPositionWS()
  n2 = Constant3Vector() {constant=<rgba 1 1 1 0>}
  n3 = CameraPositionWS()
  n4 = OneMinus(None=SphereMask(
    A=Multiply(A=@n1, B=@n2)
    B=Multiply(A=@n2, B=@n3)
    Radius=ScalarParameter() {defaultValue=1, group=Near Fade, parameterName=Fade_Distance}
    Hardness=ScalarParameter() {group=Near Fade, parameterName=hardness}
   ) {hardnessPercent=10})
  n5 = Ceil(None=@n4)
  n6 = Multiply(
   A=ScalarParameter() {defaultValue=400, parameterName=Width}
   B=DynamicParameter() {parameterIndex=1, paramNames [SplineWidthMult, Param2, Param3, Param4]}.SplineWidthMult
  )
  n7 = LinearInterpolate(A=@n6, B=ScalarParameter() {defaultValue=100, parameterName=WidthTip}, Alpha=ScalarParameter() {defaultValue=1, parameterName=Use Witdh Tip})
  n8 = Transform(None=VectorParameter() {group=Rotation, parameterName=NormalizedRotationAxis, defaultValue=<rgba 1 0 0 0>}.RGB) {transformSourceType=TRANSFORMSOURCE_Instance}
  n9 = Time() {period=5}
  n10 = ScalarParameter() {group=Rotation, parameterName=Rotation_speed}
  n11 = VectorParameter() {group=Rotation, parameterName=PivotPointBias}
  Add(
   A=MaterialFunctionCall(
    WidthBase=Multiply(A=@n5, B=@n6)
    WidthTip=Multiply(A=@n5, B=@n7)
    Expand U or V UV Channel=StaticBoolParameter() {parameterName="Expand U or V UV Channel (SB)"}
   ) {materialFunction=&SplineThicken}.WorldPositionOffset
   B=RotateAboutAxis(
    NormalizedRotationAxis=Normalize(VectorInput=@n8)
    RotationAngle=Multiply(A=@n9, B=@n10)
    PivotPoint=TransformPosition(None=@n11.RGB) {transformSourceType=TRANSFORMPOSSOURCE_Instance, transformType=TRANSFORMPOSSOURCE_World}
    Position=WorldPosition().XYZ
   )
  ) {constB=0}
  Constant() {r=0.3}
  n12 = TextureCoordinate() {vTiling=0.7}
  n13 = Constant2Vector() {r=0.1}
  n14 = TextureSample(UVs=Panner(Coordinate=Add(A=@n12, B=@n13), Time=NamedRerouteUsage()) {speedX=0.75}) {texture=&T_Tile_RainbowColor}
  n15 = ComponentMask(None=TextureCoordinate() {uTiling=2}) {r=true}
  n16 = DynamicParameter() {defaultValue=<rgba 0 1 0 0>, paramNames [ColorOffset, Reveal, Erode, ErodeHardness]}
  n17 = VectorParameter() {parameterName=Tex_One_TilingandOffset, defaultValue=<rgba 0.5 0.6 0.3 0.4>}
  n18 = Add(A=Multiply(A=TextureCoordinate(), B=AppendVector(A=@n17.R, B=@n17.R)), B=AppendVector(A=@n17.B, B=@n17.B))
  n19 = NamedRerouteUsage()
  n20 = VectorParameter() {parameterName=Tex_Two_TilingandOffset, defaultValue=<rgba 0.3 0.8 -0.05 -0.1>}
  n21 = Add(A=Multiply(A=TextureCoordinate(), B=AppendVector(A=@n20.R, B=@n20.R)), B=AppendVector(A=@n20.B, B=@n20.B))
  n22 = NamedRerouteUsage()
  n23 = Add(
   A=TextureSample(UVs=Panner(Coordinate=@n18, Time=@n19) {speedX=1}) {samplerType=SAMPLERTYPE_Grayscale, texture=/Game/Effects/Fort_Effects/Textures/Fire/T_WispyFireNoise_Soft_M}.R
   B=TextureSample(UVs=Panner(Coordinate=@n21, Time=@n22) {speedX=0.6}) {samplerType=SAMPLERTYPE_Grayscale, texture=&T_WispyFireNoise_Soft_M}.RGB
  )
  n24 = Multiply(
   A=LinearInterpolate(
    A=VectorParameter() {parameterName=ColorStart, defaultValue=<rgba 2 0 0.8665 1>}.RGB
    B=VectorParameter() {parameterName=ColorEnd, defaultValue=<rgba 0.0166 0.2474 2 1>}.RGB
    Alpha=Add(A=@n15, B=@n16.ColorOffset)
   )
   B=@n23
  )
  n25 = ParticleColor()
  MaterialFunctionCall(Emissive=Multiply(A=Add(A=@n14.RGB, B=@n24), B=@n25.RGB)) {materialFunction=&MF_FX_EmissiveAdjustments}
  n26 = ComponentMask(None=TextureCoordinate()) {r=true}
  n27 = OneMinus(None=@n26)
  n28 = TextureCoordinate()
  n29 = Multiply(A=Sine(None=ComponentMask(None=@n28) {g=true}) {period=2}, B=Sine(None=ComponentMask(None=@n28) {r=true}) {period=2})
  n30 = VectorParameter() {parameterName=Mask_TilingandOffset, defaultValue=<rgba 1 1 0 0>}
  n31 = Multiply(A=TextureCoordinate(), B=AppendVector(A=@n30.R, B=@n30.R))
  n32 = AppendVector(A=@n30.B, B=@n30.B)
  n33 = TextureSampleParameter2D(UVs=Panner(Coordinate=Add(A=@n31, B=@n32), Time=NamedRerouteUsage()) {speedX=0.4}) {parameterName=Mask, samplerType=SAMPLERTYPE_Grayscale, texture=/Game/Effects/Fort_Effects/Textures/Smoke/T_Smoke_Trail_Soft_NoClamp}
  n34 = Multiply(A=@n33.R, B=ScalarParameter() {defaultValue=1, parameterName=MaskIntensity})
  n35 = Multiply(A=ScalarParameter() {defaultValue=1, parameterName=Opacity}, B=Multiply(A=Multiply(A=@n23, B=@n29), B=Multiply(A=@n23, B=@n34)))
  n36 = DepthFade(Opacity=@n35, FadeDistance=ScalarParameter() {defaultValue=50, parameterName=FadeDistance}) {fadeDistanceDefault=10}
  n37 = Multiply(A=MaterialFunctionCall("FadeSensitivity (optional)"=ScalarParameter() {parameterName="FadeSensitivity (optional) (S)"}) {materialFunction=&FadeBasedOnViewAngle}.Result) {constB=10}
  n38 = Multiply(A=Multiply(A=ShadingPathSwitch(Default=@n36, Deferred=@n36, Forward=@n35, Mobile=@n35), B=Saturate(None=@n37)), B=@n4)
  Multiply(A=Multiply(A=Add(A=@n27, B=@n16.Reveal), B=@n38), B=SmoothStep(Min=@n16.ErodeHardness, Max=@n16.ErodeHardness, Value=@n26) {constValue=1})
  Multiply(A=@n25.A, B=@n38)
  NamedRerouteDeclaration(None=Multiply(A=Time(), B=ScalarParameter() {defaultValue=-0.5, parameterName=Speed Direction}) {constB=-1.5}) {name=Speed, nodeColor=<rgba 0.7578 1 0 1>}
 ]
}
