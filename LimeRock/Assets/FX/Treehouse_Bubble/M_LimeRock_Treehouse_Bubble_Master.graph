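# Translucent, unlit, two-sided master material for the LimeRock treehouse bubble effect on Niagara mesh particles.
# Tinted color comes from chroma-split, bump-offset panning turbulence (second layer at high quality) plus noise-varied depth-fade intersection glow, with an optional TacoTime switch.
# Opacity combines vertical UV gradient, Fresnel, camera-distance fade, particle alpha and inside/outside multipliers.
# Offset branches compute a particle-driven normal-map wobble and a large downward push beyond a hide distance.
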Material /LimeRock/Assets/FX/Treehouse_Bubble/M_LimeRock_Treehouse_Bubble_Master
graph {
 properties {
  bUsedWithNiagaraMeshParticles=true
  blendMode=BLEND_Translucent
  pixelDepthOffsetMode=PDOM_Legacy
  shadingModel=MSM_Unlit
  twoSided=true
  usedShadingModels=Unlit
 }
 nodes [
  n1 = DepthFade(FadeDistance=ScalarParameter() {defaultValue=500, parameterName=Emissive Fade Distance}) {fadeDistanceDefault=500}
  Add(B=Multiply(A=OneMinus(None=@n1), B=ScalarParameter() {defaultValue=10, parameterName=Emissive Depth Fade Mult}))
  n2 = ScalarParameter() {defaultValue=500, parameterName=Emissive Fade Distance}
  n3 = Multiply(
   A=OneMinus(None=DepthFade(FadeDistance=@n2) {fadeDistanceDefault=500})
   B=ScalarParameter() {defaultValue=10, parameterName=Emissive Depth Fade Mult}
  )
  Add(B=@n3)
  n4 = TextureCoordinate()
  n5 = Constant() {r=0.5}
  n6 = TextureCoordinate() {uTiling=3}
  n7 = Multiply(A=Time(), B=ScalarParameter() {defaultValue=0.25, parameterName=TexturePanningSpeed2})
  n8 = Subtract(
   A=TextureSample(UVs=Panner(Coordinate=@n6, Time=@n7) {speedX=0.2, speedY=0.2}) {samplerType=SAMPLERTYPE_LinearColor, texture=&T_NoisePack_LinearColor}.R
   B=TextureSample(UVs=Panner(Coordinate=@n6, Time=@n7) {speedX=-0.2, speedY=-0.2}) {samplerType=SAMPLERTYPE_LinearColor, texture=&T_NoisePack_LinearColor}.G
  )
  n9 = DepthFade(FadeDistance=LinearInterpolate(
    A=ScalarParameter() {defaultValue=100, parameterName=DepthFadeInverse}
    B=ScalarParameter() {defaultValue=500, parameterName=DepthFadeInverse2}
    Alpha=Clamp(None=@n8)
   ))
  n10 = Multiply(
   A=SphereMask(A=ComponentMask(None=@n4) {g=true}, B=@n5, Radius=@n5) {hardnessPercent=40}
   B=Add(A=OneMinus(None=@n9), B=ScalarParameter() {defaultValue=0.6, parameterName=OpacityAdd}) {constB=0.1}
  ) {constA=1}
  n11 = Multiply(
   A=Power(Base=Multiply(B=@n10) {constA=0.8}, Exp=ScalarParameter() {defaultValue=3, parameterName=DepthPower})
   B=ScalarParameter() {defaultValue=0.35, parameterName=Fullness Intensity}
  ) {constB=0.15}
  Add(A=@n11, B=@n3)
  Constant() {r=0.5}
  Constant()
  DepthFade(FadeDistance=ScalarParameter() {defaultValue=50, parameterName=DepthFade}) {fadeDistanceDefault=50}
  n12 = WorldPosition() {worldPositionShaderOffset=WPT_CameraRelative}
  n13 = ScalarParameter() {defaultValue=35000, parameterName=FadeDist_WPO}
  n14 = Constant()
  If(
   A=MaterialFunctionCall(Vector 2=ComponentMask(None=@n12.XYZ) {g=true, r=true}) {materialFunction=/Engine/Functions/Engine_MaterialFunctions02/Utility/VectorLength}.V2 Length
   B=Multiply(A=@n13, B=ScalarParameter() {defaultValue=1.4, parameterName=HideDistanceBuffer}) {constB=1.36}
   "A > B"=Constant3Vector() {constant=<rgba 0 0 -800000 1>}
   "A == B"=@n14
   "A < B"=@n14
  )
  n15 = Add(A=ComponentMask(None=TextureCoordinate()) {g=true}, B=ScalarParameter() {defaultValue=0.7, parameterName=BottomTopBias})
  n16 = ScalarParameter() {defaultValue=15, parameterName=BottomTopExponent}
  n17 = MaterialFunctionCall() {materialFunction=/Engine/Functions/Engine_MaterialFunctions02/WorldPositionOffset/ObjectScale}
  n18 = Multiply(
   A=Divide(
    A=MaterialFunctionCall!(In=@n17.Scale X) {materialFunction=/Engine/Functions/Engine_MaterialFunctions02/Utility/Round}.Result
    B=ScalarParameter() {defaultValue=2, parameterName=Tile UV Scale}
   )
   B=TextureCoordinate()
  )
  n19 = AppendVector(
   A=ScalarParameter() {defaultValue=4, parameterName=U Repeat Caustic 1}
   B=ScalarParameter() {defaultValue=2.2215, parameterName=V Repeat Caustic 1}
  )
  n20 = Time()
  n21 = ScalarParameter() {parameterName=TimeAdd}
  n22 = Add(A=@n20, B=@n21)
  n23 = TextureSample(UVs=Panner(Coordinate=Multiply(A=@n18, B=@n19), Time=@n22) {bFractionalPart=true, speedX=0.0268, speedY=-0.0194}) {samplerType=SAMPLERTYPE_Normal, texture=/Game/VisualThreatManager/StormVisuals/Textures/2ndInfectedMapNormal}
  n24 = AppendVector(
   A=ScalarParameter() {defaultValue=2, parameterName=U Repeat Caustic 2}
   B=ScalarParameter() {defaultValue=0.62, parameterName=V Repeat Caustic 2}
  )
  n25 = TextureSample(UVs=Panner(Coordinate=Multiply(A=@n18, B=@n24), Time=@n22) {bFractionalPart=true, speedX=-0.0226, speedY=-0.0104}) {samplerType=SAMPLERTYPE_Normal, texture=&2ndInfectedMapNormal}
  n26 = Multiply(A=Add(A=@n23.RGB, B=@n25.RGB), B=VertexNormalWS())
  n27 = VectorParameter() {parameterName=WPO Strength, defaultValue=<rgba 2 2 2 0>}
  Multiply(
   A=Clamp(None=Power(Base=@n15, Exp=@n16))
   B=Multiply(A=Multiply(A=@n26, B=@n27.RGB), B=DynamicParameter() {paramNames [Uniform WPO Strength, Param2, Param3, Param4]}.Uniform WPO Strength)
  )
  n28 = AppendVector(A=ScalarParameter() {parameterName=Chroma_UOffset}, B=ScalarParameter() {defaultValue=-0.05, parameterName=Chroma_VOffset2})
  n29 = TextureCoordinate()
  n30 = AppendVector(A=ScalarParameter() {defaultValue=6, parameterName=TextureA_UScale}, B=ScalarParameter() {defaultValue=2, parameterName=TextureA_VScale})
  n31 = TextureCoordinate() {uTiling=3}
  n32 = Multiply(A=Time(), B=ScalarParameter() {defaultValue=0.025, parameterName=TexturePanningSpeed})
  n33 = TextureSample(UVs=Panner(Coordinate=@n31, Time=@n32) {speedX=0.2, speedY=0.2}) {samplerType=SAMPLERTYPE_LinearColor, texture=&T_NoisePack_LinearColor}
  n34 = TextureSample(UVs=Panner(Coordinate=@n31, Time=@n32) {speedX=-0.2, speedY=-0.2}) {samplerType=SAMPLERTYPE_LinearColor, texture=&T_NoisePack_LinearColor}
  n35 = ComponentMask(None=TextureCoordinate() {uTiling=0.5, vTiling=0.5}) {g=true}
  n36 = MaterialFunctionCall(In=Abs(None=Sine(None=@n35)), Contrast=Constant() {r=0.1}) {materialFunction=&CheapContrast}
  n37 = VertexNormalWS()
  n38 = CameraVectorWS()
  n39 = CurveAtlasRowParameter(CurveTime=Abs(None=DotProduct(A=@n37, B=@n38))) {
   atlas=&AthenaCurveAtlas01
   curve=/Game/Athena/Environments/Nexus/Materials/FXshader/Curve_Ramps
   defaultValue=79
  }
  n40 = ScalarParameter() {defaultValue=0.025, parameterName=TexturePanningSpeed}
  n41 = Add(
   A=Multiply(A=Subtract(A=@n33.R, B=@n34.G), B=@n36.Result)
   B=Multiply(
    A=Subtract(A=@n39.R)
    B=MaterialFunctionCall(Speed=@n40) {materialFunction=&DebugTimeSine}.Result
   )
  )
  n42 = Multiply(A=@n36.Result, B=ScalarParameter() {defaultValue=0.25, parameterName=Height})
  n43 = TextureCoordinate() {uTiling=9, vTiling=3}
  n44 = Multiply(A=Time(), B=ScalarParameter() {defaultValue=0.025, parameterName=TexturePanningSpeed})
  n45 = Multiply(
   A=TextureSample(UVs=Panner(Coordinate=@n43, Time=@n44) {speedX=-0.02, speedY=-0.03}) {samplerType=SAMPLERTYPE_Grayscale, texture=&LowResBlurredNoise}.R
   B=Constant2Vector() {g=0.14, r=0.07}
  )
  n46 = Add(A=BumpOffset(Coordinate=Multiply(A=@n29, B=@n30), Height=@n41, HeightRatioInput=@n42), B=@n45)
  n47 = Multiply(A=Time(), B=ScalarParameter() {defaultValue=0.025, parameterName=TexturePanningSpeed})
  n48 = Panner(Coordinate=Add(A=@n46, B=@n28), Time=@n47) {speedY=-0.3}
  n49 = Panner(Coordinate=@n46, Time=@n47) {speedY=-0.3}
  n50 = Add(A=@n46, B=Multiply(A=@n28) {constB=-1})
  n51 = AppendVector(
   A=AppendVector(
    A=TextureSample(UVs=@n48) {texture=/Game/Global/Textures/Default/Caustics/T_Turbulence_Seamless}.R
    B=TextureSample(UVs=@n49) {texture=&T_Turbulence_Seamless}.G
   )
   B=TextureSample(UVs=Panner(Coordinate=@n50, Time=@n47) {speedY=-0.3}) {texture=&T_Turbulence_Seamless}.B
  )
  n52 = AppendVector(A=ScalarParameter() {parameterName=Chroma_UOffset}, B=ScalarParameter() {defaultValue=0.025, parameterName=Chroma_VOffset})
  n53 = TextureCoordinate()
  n54 = AppendVector(A=ScalarParameter() {defaultValue=8, parameterName=TextureB_UScale}, B=ScalarParameter() {defaultValue=1.4, parameterName=TextureB_VScale})
  n55 = Add(A=@n45, B=BumpOffset(Coordinate=Multiply(A=@n53, B=@n54), Height=@n41, HeightRatioInput=@n42))
  n56 = Add(A=@n55, B=@n52)
  n57 = Multiply(A=Time(), B=ScalarParameter() {defaultValue=0.025, parameterName=TexturePanningSpeed})
  n58 = AppendVector(
   A=TextureSample(UVs=Panner(Coordinate=@n56, Time=@n57) {speedY=-0.3}) {texture=&T_Turbulence_Seamless}.R
   B=TextureSample(UVs=Panner(Coordinate=@n55, Time=@n57) {speedY=-0.3}) {texture=&T_Turbulence_Seamless}.G
  )
  n59 = Multiply(A=@n52) {constB=-1}
  n60 = TextureSample(UVs=Panner(Coordinate=Add(A=@n55, B=@n59), Time=@n57) {speedY=-0.3}) {texture=&T_Turbulence_Seamless}
  n61 = QualitySwitch(Default=@n51, High=LinearInterpolate(A=@n51, B=AppendVector(A=@n58, B=@n60.B)))
  n62 = FeatureLevelSwitch(Default=@n51, SM4_REMOVED=@n51, SM5=@n61, SM6=@n61)
  n63 = Add(A=@n62, B=@n11)
  n64 = QualitySwitch(Default=@n63, High=@n63)
  n65 = FeatureLevelSwitch(Default=@n64, SM4_REMOVED=@n62, SM5=@n64, SM6=@n64)
  n66 = ComponentMask(None=TextureCoordinate()) {g=true}
  n67 = ScalarParameter() {defaultValue=1.5, parameterName=Top Edge Alpha Contrast}
  n68 = Saturate(None=Multiply(A=Power(Base=@n66, Exp=@n67)) {constB=5})
  n69 = Clamp(None=MaterialFunctionCall(Distance=ScalarParameter() {defaultValue=20000, parameterName=FadeDist}) {materialFunction=&Near_Camera_Fade}.Result) {minDefault=0.1}
  n70 = Fresnel(ExponentIn=ScalarParameter() {defaultValue=2, parameterName=Exterior Alpha Fresnel Exp})
  n71 = ReflectionVectorWS()
  n72 = VertexNormalWS()
  n73 = Clamp(None=Ceil(None=DotProduct(A=@n71, B=@n72)))
  n74 = LinearInterpolate(Alpha=Fresnel(ExponentIn=ScalarParameter() {defaultValue=0.5, parameterName=UVFresnelExp}) {baseReflectFraction=0, exponent=1}) {constA=1.6, constB=-0.15}
  n75 = Multiply(A=Multiply(A=Multiply(A=@n65, B=@n68), B=LinearInterpolate(A=@n69, B=@n70, Alpha=@n73) {constA=1}), B=Saturate(None=@n74))
  n76 = ScalarParameter() {parameterName=Opacity Lerp}
  n77 = MaterialFunctionCall(Distance=@n13) {materialFunction=&Near_Camera_Fade}
  n78 = Clamp(None=LinearInterpolate(Alpha=@n77.Result) {constA=1.25, constB=0})
  n79 = LinearInterpolate(
   A=ScalarParameter() {defaultValue=0.1289, parameterName=OpacityMult_inside}
   B=ScalarParameter() {defaultValue=1.7864, parameterName=OpacityMult}
   Alpha=Saturate(None=TwoSidedSign())
  )
  Multiply(A=ParticleColor().A, B=Multiply(A=LinearInterpolate(A=@n75, B=@n74, Alpha=@n76), B=Multiply(A=@n78, B=@n79)))
  n80 = ScalarParameter() {defaultValue=1, parameterName=Opacity_Day}
  Multiply(
   A=@n75
   B=MaterialFunctionCall(Color_Morning=@n80, Color_Day=@n80, Color_Evening=@n80, Color_Night=ScalarParameter() {defaultValue=0.35, parameterName=Opacity}) {materialFunction=&Color_Time_of_Day}.Result
  )
  n81 = ScalarParameter() {defaultValue=0.01, parameterName=Sine Speed A}
  n82 = Add(A=@n23.RGB, B=Add(A=@n21, B=Multiply(A=@n20, B=@n81)))
  n83 = ScalarParameter() {defaultValue=0.0785, parameterName=Sine Speed B}
  n84 = Add(A=@n25.RGB, B=Add(A=@n21, B=Multiply(A=@n20, B=@n83)))
  n85 = Desaturation(None=Multiply(A=Sine(None=@n82) {period=2.22}, B=Sine(None=@n84) {period=4.4})) {luminanceFactors=<rgba 1 1 1 1>}
  OneMinus(None=ConstantBiasScale(None=@n85) {bias=2, scale=0.25})
  OneMinus(None=@n77.Result)
  n86 = Abs(None=@n85)
  Power(Base=Clamp(None=OneMinus(None=@n86)), Exp=ScalarParameter() {defaultValue=3, parameterName=Caustic Exponent})
  n87 = Multiply(
   A=Multiply(A=VectorParameter() {parameterName=TextureTint, defaultValue=<rgba 0.2 0.2267 1 0>}.RGB, B=@n65)
   B=ScalarParameter() {defaultValue=15, parameterName=BubbleBaseIntensity}
  )
  Reroute(None=@n87)
  StaticSwitchParameter(
   True=MaterialFunctionCall!(StartingHue, Base Emissive=@n87, Offset) {materialFunction=/Game/Environments/Props/RiftBeacon/FX/RiftBubble_GreenGoop}.TacoTimeResults
   False=@n87
  ) {parameterName=Allow TacoTime}
 ]
}
