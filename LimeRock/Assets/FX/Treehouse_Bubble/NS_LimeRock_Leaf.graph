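# Niagara effect that throws a gust of green leaf sprites along +X, apparently for the LimeRock Treehouse Bubble effect.
# One GPU emitter, Small_Debris, pairs an instant burst with a spawn rate that tapers late in the loop.
# Particles start in an offset, flattened-sphere region with random green tint, size and fast velocity.
# Curl noise and drag shape their flight while they spin, animate through SubUV frames, and scale and fade in and out.
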NiagaraSystem /LimeRock/Assets/FX/Treehouse_Bubble/NS_LimeRock_Leaf
dependencies {
 usedDataInterfaces [&NiagaraDataInterfaceCurve, &NiagaraDataInterfaceSpriteRendererInfo]
 usedDynamicInputs [&FindQuatBetweenTwoVectors]
 usedModules [&SystemState]
}
emitters {Small_Debris {
  simTarget=GPUComputeSim
  emitterUpdateScript [
   &EmitterState(Life Cycle Mode=::NewEnumerator1, Loop Behavior=::NewEnumerator1, UseLoopDelay=false)
   &SpawnBurst_Instantaneous(Spawn Count=100, Use Spawn Probability=false, UseLoopCountLimit=false)
   &SpawnRate(
    SpawnRate=&Multiply_Float(
     A=200
     B=&FloatFromCurve(
      FloatCurve=DataInterface() {propertyValues {Curve {keys [{interpMode=RCIM_Cubic, tangentMode=RCTM_Break, time=0.7, value=1}, {interpMode=RCIM_Cubic, tangentMode=RCTM_Break, time=1, value=0}]}}}
      CurveIndex=$Emitter.NormalizedLoopAge
     )
    )
    Use Spawn Probability=false
   )
  ]
  particleSpawnScript [
   &InitializeParticle(
    Color=&RandomRangeLinearColor(Minimum=<rgba 0.0163 0.1667 0.0079 1>, Maximum=<rgba 0.2184 0.8958 0.028 1>)
    Lifetime Max=4
    Lifetime Min=3.2
    Lifetime Mode=::NewEnumerator1
    Sprite Rotation Mode=::NewEnumerator2
    Sprite Size Mode=::NewEnumerator3
    Uniform Sprite Size=&Multiply_Float(A=1, B=&RandomRangeFloat(Minimum=70, Maximum=100))
    UsePositionOffset=false
    Write Material Random=true
   )
   &ShapeLocation(
    MaskSpawnGroup=false
    Non Uniform Scale=<1.2 1.2 1>
    Offset=<-700 0 1400>
    Offset Coordinate Space=::World
    Offset Mode=::NewEnumerator0
    Override Seed=false
    Rotation Coordinate Space=::World
    Shape Origin=/Niagara/DynamicInputs/Helpers/SimulationPosition()
    Sphere Radius=600
    UseOwnerScale=false
   ) {module=ShapeLocation001}
   &AddVelocity(
    Rotation Coordinate Space=::World
    Velocity=&RandomRangeVector(Minimum=<5000 -1900 -1900>, Maximum=<8000 1900 1900>)
   )
  ]
  particleUpdateScript [
   &ParticleState()
   &ScaleSpriteSize(Uniform Curve Sprite Scale=DataInterface() {propertyValues {Curve {keys [
        {interpMode=RCIM_Cubic, tangentMode=RCTM_User, time=0, value=0}
        {interpMode=RCIM_Cubic, tangentMode=RCTM_User, time=0.5, value=1}
        {interpMode=RCIM_Cubic, tangentMode=RCTM_User, time=0.8, value=0.95}
        {tangentMode=RCTM_User, time=1, value=0}
       ]}}})
   &SpriteRotationRate(Rotation Rate=&RandomRangeFloat(Minimum=-600, Maximum=600))
   &CurlNoiseForce(Bias Noise=false, Noise Strength=500, Pan Noise=false)
   &Drag(
    Drag=&RandomRangeFloat(Minimum=0.2, Maximum=0.3)
    Rotational Drag:NiagaraFloat=&RandomRangeFloat(Minimum=0.2, Maximum=0.3)
    Use Linear Drag=true
    Use Rotational Drag=true
   )
   &SolveForcesAndVelocity(Clamp Velocity=false, Limit Acceleration=false) {module=SolveForcesAndVelocity001}
   &ScaleColor(
    Scale Alpha=&FloatFromCurve(
     FloatCurve=DataInterface() {propertyValues {Curve {keys [
         {interpMode=RCIM_Cubic, tangentMode=RCTM_User, time=0, value=0}
         {interpMode=RCIM_Cubic, tangentMode=RCTM_User, time=0.1, value=1}
         <at 0.8 0.95>
         {tangentMode=RCTM_User, time=1, value=0}
        ]}}}
     Scale Curve=1.5
    )
    ScaleA=true
    ScaleRGB=true
   )
   &SubUVAnimation(
    Playback Scale Curve=DataInterface() {propertyValues {Curve {keys [{interpMode=RCIM_Cubic, tangentMode=RCTM_User, time=0, value=0}, {interpMode=RCIM_Cubic, tangentMode=RCTM_User, time=1, value=1}]}}}
    "Sprite Renderer "=DataInterface() {propertyValues {SpriteRenderer=&self:Small_Debris_0.NiagaraSpriteRendererProperties_0}}
    SubUV Animation Mode:ENiagara_SubUVLookupModeV2=::NewEnumerator3
    UseEndFrame=true
    UsePlayRateScale=true
    UseStartFrame=true
   ) {module=SubUVAnimation002}
  ]
  rendererClasses [&NiagaraSpriteRendererProperties]
  renderers [{rendererClass=&NiagaraSpriteRendererProperties}]
 }}
summary {
 systemName=NS_LimeRock_Leaf
 emitters [{emitterName=Small_Debris, simTarget=GPUComputeSim, rendererClasses [&NiagaraSpriteRendererProperties]}]
 userVariables []
}
