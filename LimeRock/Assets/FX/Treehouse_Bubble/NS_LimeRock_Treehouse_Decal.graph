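# Niagara system that projects a large decal for the LimeRock Treehouse Bubble effect.
# A single Decal emitter burst-spawns a fixed-size decal oriented along X, with a random material value and no kill on lifetime expiry.
# Each update scales particle color from two user floats: Opacity drives RGB and V_Mask drives alpha, so the color channels carry external controls to the decal renderer.
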NiagaraSystem /LimeRock/Assets/FX/Treehouse_Bubble/NS_LimeRock_Treehouse_Decal
dependencies {
 usedDataInterfaces [&NiagaraDataInterfaceCurve]
 usedDynamicInputs [&FloatFromCurve]
 usedModules [
  &SystemState
  &self:SystemSpawnScript.SystemScriptSource.SystemScriptGraph.NiagaraNodeAssignment_0.SetVariables_#1
  &self:SystemSpawnScript.SystemScriptSource.SystemScriptGraph.NiagaraNodeAssignment_1.SetVariables_#2
 ]
}
emitters {Decal {
  emitterUpdateScript [
   &EmitterState()
   &SpawnBurst_Instantaneous(Use Spawn Probability=false, UseLoopCountLimit=false)
  ]
  particleSpawnScript [
   &InitializeParticle(Lifetime=4.5, UsePositionOffset=false, Write Material Random=true)
   /SetVariables(Particles.DecalOrientation:Quat4f=&OrientMeshToVector(Look At Direction=<1 0 0>, Reference Axis Vector=<1 0 0>)) {isSetParameters=true}
   /SetVariables(Particles.DecalSize:Vector3f=<3000 3000 3000>) {isSetParameters=true}
  ]
  particleUpdateScript [
   &ParticleState(Kill Particles When Lifetime Has Elapsed=false)
   &ScaleColor(Scale Alpha=$User.V_Mask, Scale RGB=&VectorFromFloat(Value=$User.Opacity), ScaleA=true, ScaleRGB=true)
  ]
  rendererClasses [&NiagaraDecalRendererProperties]
  renderers [{rendererClass=&NiagaraDecalRendererProperties, rendererIndex=0}]
 }}
summary {
 systemName=NS_LimeRock_Treehouse_Decal
 emitters [{emitterName=Decal, rendererClasses [&NiagaraDecalRendererProperties]}]
 userVariables [
  User.Opacity {type=NiagaraFloat, defaultValue=1, description=""}
  User.V_Mask {type=NiagaraFloat, defaultValue=1, description=""}
 ]
}
