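# Niagara burst effect for suppressed energy at the LimeRock treehouse, apparently part of its bubble FX.
# Three mesh emitters are active: a horizontal shockwave, rotating soft airflow and a twice-fired inward energy wave.
# Each shrinks over its lifetime by User.Uniform Scale and curves, which also animate alpha and dynamic material parameters.
# Flare, sprite, burst, offset and ribbon tether emitters are disabled, leaving User.TreePosition without an active consumer.
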NiagaraSystem /LimeRock/Assets/FX/Treehouse_Bubble/NS_LimeRock_Treehouse_Energy_Suppressed
dependencies {
 usedDataInterfaces [
  &NiagaraDataInterfaceCurve
  &NiagaraDataInterfaceColorCurve
  &NiagaraDataInterfaceVector2DCurve
  &NiagaraDataInterfaceVectorCurve
 ]
 usedDynamicInputs [
  &RandomRangeLinearColor
  &SimulationPosition
  &UniformRangedFloat
 ]
 usedModules [&SystemState]
}
emitters {
 Burst_Emitter {
  bEnabled=false
  emitterUpdateScript [
   &EmitterState(Life Cycle Mode=::NewEnumerator1, Loop Behavior=::NewEnumerator1, Loop Duration=6, UseLoopDelay=false)
   &SpawnBurst_Instantaneous(Use Spawn Probability=false, UseLoopCountLimit=false)
  ]
  particleSpawnScript [
   &InitializeParticle(
    Color Mode=::NewEnumerator0
    Lifetime=6
    Mesh Scale Mode=::NewEnumerator3
    Position Mode=::NewEnumerator0
    Sprite Size Mode=::NewEnumerator3
    Uniform Sprite Size=2
    Write Material Random=true
   )
   &ApplyInitialForces() {enabled=false}
   &InitialMeshOrientation(
    Mesh Orientation Mode=::NewEnumerator1
    Rotation=&MakeVector(X=&ReturnNormalizedExecIndex(), Y=0, Z=0)
    Use Rotation Vector=true
   ) {enabled=false}
   &SystemLocation(MaskSpawnGroup=false, Offset:NiagaraPosition=$User.TreePosition)
  ]
  particleUpdateScript [
   &ParticleState()
   &ScaleMeshSize(Scale Factor=<1 1 0.2>) {enabled=false, module=ScaleMeshSize001}
   &ScaleMeshSize(Scale Factor=&VectorFromFloat(Value=&Multiply_Float(A=1, B=$User.Uniform Scale))) {enabled=false}
   &DynamicMaterialParameters(
    Index 0 Param 2=&FloatFromCurve(FloatCurve=DataInterface() {propertyValues {Curve {keys [
         {arriveTangent=2.6347, interpMode=RCIM_Cubic, leaveTangent=2.6347, tangentMode=RCTM_User, time=0, value=-1}
         {arriveTangent=2.8508, interpMode=RCIM_Cubic, leaveTangent=2.8508, tangentMode=RCTM_User, time=0.4484, value=0.2717}
         <at 1 1 RCIM_Cubic>
        ]}}})
    Index 1 Param 1=1
    Module.Index0Float0Write=true
    Module.Index0Float1Write=true
    Module.Index0Float2Write=true
    Module.Index0Float3Write=true
    Module.Index1Float0Write=true
    Module.Index1Float1Write=true
    Module.Index1Float2Write=true
    Module.Index1Float3Write=true
    Write Parameter Index 1=::NewEnumerator1
   )
   &ScaleColor(
    Scale Alpha=&FloatFromCurve(FloatCurve=DataInterface() {propertyValues {Curve {
        defaultValue=0
        keys [
         {tangentMode=RCTM_User, time=0, value=1}
         {arriveTangent=-0.0405, interpMode=RCIM_Cubic, leaveTangent=-0.0405, tangentMode=RCTM_User, time=0.89, value=1}
         {arriveTangent=-8.0089, interpMode=RCIM_Cubic, leaveTangent=-8.0089, tangentMode=RCTM_User, time=1, value=0}
        ]
       }}})
    ScaleA=true
    ScaleRGB=true
   ) {enabled=false}
  ]
  rendererClasses [&NiagaraMeshRendererProperties]
  renderers [{rendererClass=&NiagaraMeshRendererProperties, rendererIndex=0}]
 }
 ElectricSpirtes {
  bEnabled=false
  emitterUpdateScript [
   &EmitterState()
   &SpawnBurst_Instantaneous(Spawn Count=12, Use Spawn Probability=false, UseLoopCountLimit=false)
   &SpawnBurst_Instantaneous(Spawn Count=8, Spawn Time=0.3, Use Spawn Probability=false, UseLoopCountLimit=false) {enabled=false, module=SpawnBurst_Instantaneous002}
   &SpawnBurst_Instantaneous(Spawn Count=5, Spawn Group=1, Spawn Time=0.6, Use Spawn Probability=false, UseLoopCountLimit=false) {enabled=false, module=SpawnBurst_Instantaneous001}
  ]
  particleSpawnScript [
   &InitializeParticle(
    Color Maximum=<rgba 7 7 7 1>
    Color Minimum=<rgba 3 3 3 1>
    Color Mode=::NewEnumerator2
    Lifetime Max=0.3
    Lifetime Min=0.2
    Lifetime Mode=::NewEnumerator1
    Mass Mode=::NewEnumerator1
    Sprite Rotation Mode=::NewEnumerator2
    Sprite Size=&UniformRangedVector2D(Minimum=<2 5>, Maximum=<40 60>)
    Sprite Size Mode=::NewEnumerator4
    UsePositionOffset=false
    Write Material Random=true
   ) {module=InitializeParticle001}
   &DynamicMaterialParameters(
    Index 0 Param 1=$Particles.MaterialRandom
    Index 0 Param 2=&Lerp_Float(A=0.9, B=1.25, Alpha=$Particles.MaterialRandom)
    Index 0 Param 4=1
    Module.Param0WriteEnabled=true
    Module.Param1WriteEnabled=true
    Module.Param2WriteEnabled=false
    Module.Param3WriteEnabled=true
   )
   &ShapeLocation(MaskSpawnGroup=false, Override Seed=false, Sphere Radius=1000, Sphere Surface Distribution=1, UseOwnerScale=true)
  ]
  particleUpdateScript [
   &ParticleState()
   &SpriteRotationRate(Rotation Rate=&RandomRangeFloat(Minimum=-100, Maximum=100))
   &DynamicMaterialParameters(
    Index 0 Param 3=&FloatFromCurve(FloatCurve=DataInterface() {propertyValues {Curve {keys [<at 0.5 0 RCIM_Cubic>, {arriveTangent=3.7979, interpMode=RCIM_Cubic, leaveTangent=3.7979, tangentMode=RCTM_User, time=1, value=1}]}}})
    Module.Param0WriteEnabled=false
    Module.Param0WriteEnabled1=true
    Module.Param1WriteEnabled=false
    Module.Param1WriteEnabled1=true
    Module.Param2WriteEnabled=true
    Module.Param2WriteEnabled1=true
    Module.Param3WriteEnabled=false
    Module.Param3WriteEnabled1=true
    Write Parameter Index 1=true
   ) {module=DynamicMaterialParameters001}
   &ScaleSpriteSize(
    Uniform Curve Scale=&MaskFloatBySpawnGroup(Masked Float=1, Passthrough Float=0.5, Spawn Group Mask=1)
    Uniform Curve Sprite Scale=DataInterface() {propertyValues {Curve {keys [{arriveTangent=0.4536, interpMode=RCIM_Cubic, leaveTangent=0.4536, tangentMode=RCTM_User, time=0, value=0.65}, <at 1 1>]}}}
   )
   &ScaleSpriteSize(Scale Sprite Size Mode=::NewEnumerator0, Uniform Scale Factor=&Multiply_Float(A=10, B=$User.Uniform Scale)) {module=ScaleSpriteSize001}
   &CameraOffset(Camera Offset Amount=/Niagara/DynamicInputs/Math/CalculateParticleRadius(
     Radius Calculation Type:ENiagaraCollisionRadiusOptions=::NewEnumerator0
     Particle Radius:NiagaraFloat=1
     Mesh Dimensions:Vector3f=<1 1 1>
     Method for Calculating Particle Radius:ENiagaraCalculateRadiusOptions=::NewEnumerator1
    ))
   &ScaleColor(
    Linear Color Curve=DataInterface() {propertyValues {
      AlphaCurve {keys [<at 0 1>,<at 1 0>]}
      BlueCurve {keys [<at 0.39 1>,<at 0.9147 0.7058>]}
      GreenCurve {keys [<at 0.39 0.4118>,<at 0.9147 0.1222>]}
      RedCurve {keys [<at 0.39 0.2031>,<at 0.9147 1>]}
     }}
    Scale Mode=::NewEnumerator2
   )
  ]
  rendererClasses [&NiagaraSpriteRendererProperties]
  renderers [{rendererClass=&NiagaraSpriteRendererProperties}]
 }
 Energy_Offset001 {
  bEnabled=false
  emitterUpdateScript [
   &EmitterState()
   &SpawnBurst_Instantaneous(Spawn Group=1, Spawn Time=0.6, Use Spawn Probability=false, UseLoopCountLimit=false) {module=SpawnBurst_Instantaneous003}
   &SpawnBurst_Instantaneous(Spawn Time=0.3, Use Spawn Probability=false, UseLoopCountLimit=false) {module=SpawnBurst_Instantaneous002}
   &SpawnBurst_Instantaneous(Use Spawn Probability=false, UseLoopCountLimit=false) {module=SpawnBurst_Instantaneous001}
   /SetVariables(Emitter.Linear Color:LinearColor=&LerpLinearColors(
     StartColor=<rgba 0.0159 0.1401 1 1>
     EndColor=<rgba 0.7387 0.1312 1 1>
     LerpFactor=&Waveform("Waveform [1]"=::NewEnumerator0, "[1] Frequency"=2, "[1] Amplitude Min/Max"=<0 1>)
    )) {isSetParameters=true}
  ]
  particleSpawnScript [
   &InitializeParticle(
    Color=<rgba 1 1 1 0.6>
    Lifetime=0.8
    Mesh Scale=&Multiply_VectorByFloat(Vector=<0.5 0.5 1>, Float=25)
    Mesh Scale Mode=::NewEnumerator4
    UsePositionOffset=true
    Write Material Random=true
   )
   &InitialMeshOrientation(Mesh Axis To Orient=<0 0 1>, Mesh Orientation Mode=::NewEnumerator2, Use Rotation Vector=true)
   /SetVariables(Particles.MeshOrientation=&MultiplyQuaternion(
     Quaternion A=$Particles.InitialRotation
     Quaternion B=&MakeQuatFromAxisAngle(Axis=<0 0 1>, AngleInDegrees=&RandomRangeFloat(Minimum=-360, Maximum=360), NormalizeAxis=true)
    )) {isSetParameters=true}
  ]
  particleUpdateScript [
   &ParticleState()
   &ScaleMeshSize(Scale Factor=&VectorFromFloat(Value=&FloatFromCurve(
      FloatCurve=DataInterface() {propertyValues {Curve {defaultValue=0, keys [{tangentMode=RCTM_User, time=0, value=1}, {tangentMode=RCTM_User, time=1, value=0}]}}}
      Scale Curve=&Multiply_Float(A=0.5, B=$User.Uniform Scale)
     )))
   &DynamicMaterialParameters(
    Index 0 Param 1=&FloatFromCurve(FloatCurve=DataInterface() {propertyValues {Curve {defaultValue=0.5, keys [{tangentMode=RCTM_User, time=0, value=0}, {tangentMode=RCTM_User, time=1, value=1}]}}})
    Index 0 Param 2=&RandomRangeFloat(Minimum=-2, Maximum=1)
    Index 0 Param 3=0.1
    Module.Index0Param0Write=true
    Module.Index0Param1Write=true
    Module.Index0Param2Write=true
    Module.Index0Param3Write=false
   ) {module=DynamicMaterialParameters001}
   &ScaleColor(
    Scale Alpha=&FloatFromCurve(
     FloatCurve=DataInterface() {propertyValues {Curve {keys [<at 0 0>,<at 0.3 1>,<at 1 0>]}}}
     Scale Curve=&FloatFromCurve(FloatCurve=DataInterface() {propertyValues {Curve {keys [<at 0 0.5>,<at 0.2 1>]}}})
    )
    Scale RGB=&VectorFromFloat(Value=&FloatFromCurve(FloatCurve=DataInterface() {propertyValues {Curve {keys [<at 0 0>,<at 0.3 1>,<at 1 0>]}}}))
    ScaleA=true
    ScaleRGB=true
   )
   &UpdateMeshOrientation(Rotation Rate=0.025, Rotation Vector=<0 0 1>)
   &CameraOffset(Camera Offset Amount=-350)
  ]
  rendererClasses [&NiagaraMeshRendererProperties]
  renderers [{rendererClass=&NiagaraMeshRendererProperties, rendererIndex=0}]
 }
 Flare_Complex {
  bEnabled=false
  emitterUpdateScript [
   &EmitterState()
   &SpawnBurst_Instantaneous(Use Spawn Probability=false, UseLoopCountLimit=false)
   &SpawnBurst_Instantaneous(Spawn Time=0.3, Use Spawn Probability=false, UseLoopCountLimit=false) {module=SpawnBurst_Instantaneous002}
   &SpawnBurst_Instantaneous(Spawn Group=1, Spawn Time=0.6, Use Spawn Probability=false, UseLoopCountLimit=false) {module=SpawnBurst_Instantaneous001}
  ]
  particleSpawnScript [
   &InitializeParticle(
    Color=&MaskLinearColorBySpawnGroup(Masked Linear Color=<rgba 2 2 2 0.25>, Passthrough Linear Color=<rgba 2 2 2 0.3>, Spawn Group Mask=1)
    Lifetime=&MaskFloatBySpawnGroup(Masked Float=0.5, Passthrough Float=0.2, Spawn Group Mask=1)
    Sprite Size Mode=::NewEnumerator3
    Uniform Sprite Size=&MaskFloatBySpawnGroup(Masked Float=200, Passthrough Float=160, Spawn Group Mask=1)
    UsePositionOffset=false
    Write Material Random=true
   )
   &CameraOffset(Camera Offset Amount=15)
  ]
  particleUpdateScript [
   &ParticleState()
   &ScaleSpriteSize(Scale Sprite Size Mode=::NewEnumerator0, Uniform Scale Factor=&Multiply_Float(A=7, B=$User.Uniform Scale)) {module=ScaleSpriteSize001}
   &ScaleSpriteSize(
    Uniform Curve Scale=&MakeFloatFromVector(Vector=$Engine.Owner.Scale, Channel=::NewEnumerator0)
    Uniform Curve Sprite Scale=DataInterface() {propertyValues {Curve {
       defaultValue=0
       keys [
        {interpMode=RCIM_Cubic, tangentMode=RCTM_User, time=0, value=1}
        {arriveTangent=-3.0723, interpMode=RCIM_Cubic, leaveTangent=-3.0723, tangentMode=RCTM_User, time=1, value=0}
       ]
      }}}
   )
   &DynamicMaterialParameters(
    Index 0 Param 1=0.6
    Index 0 Param 2=&FloatFromCurve(FloatCurve=DataInterface() {propertyValues {Curve {
        defaultValue=1
        keys [
         {arriveTangent=0.3652, interpMode=RCIM_Cubic, time=0, value=1}
         {arriveTangent=-0.9919, interpMode=RCIM_Cubic, leaveTangent=-0.9919, tangentMode=RCTM_User, time=1, value=0.5}
        ]
       }}})
    Index 3 Param 1=2
    Index 3 Param 2=5
    Module.Index0Float0Write=true
    Module.Index0Float1Write=true
    Module.Index0Float2Write=false
    Module.Index0Float3Write=false
    Module.Index3Float0Write=true
    Module.Index3Float1Write=true
    Module.Index3Float2Write=false
    Module.Index3Float3Write=false
    Write Parameter Index 3=::NewEnumerator1
   )
   &ScaleColor(
    Linear Color Curve=DataInterface() {propertyValues {
      AlphaCurve {keys [<at 0 1>,<at 1 0>]}
      BlueCurve {keys [<at 0 1>,<at 0.2882 1>,<at 0.6648 1>,<at 0.9147 0.7058>]}
      GreenCurve {keys [<at 0 0.3417>,<at 0.2882 0.0218>,<at 0.6648 0.3092>,<at 0.9147 0.1222>]}
      RedCurve {keys [<at 0 0.1853>,<at 0.2882 0.7831>,<at 0.6648 0.0641>,<at 0.9147 1>]}
     }}
    Scale Mode=::NewEnumerator2
   ) {module=ScaleColor001}
   &ScaleColor(
    Scale Alpha=&FloatFromCurve(FloatCurve=DataInterface() {propertyValues {Curve {keys [
         {arriveTangent=23.9365, interpMode=RCIM_Cubic, leaveTangent=23.9365, tangentMode=RCTM_User, time=0, value=0}
         {arriveTangent=1.6667, interpMode=RCIM_Cubic, leaveTangent=1.6667, time=0.1, value=1}
         {arriveTangent=-0.1118, interpMode=RCIM_Cubic, leaveTangent=13.5262, tangentMode=RCTM_Break, time=0.3, value=0.5}
         {arriveTangent=-0.5125, interpMode=RCIM_Cubic, leaveTangent=-0.5125, tangentMode=RCTM_User, time=0.4, value=1}
         {arriveTangent=-2.0538, interpMode=RCIM_Cubic, leaveTangent=-2.0538, tangentMode=RCTM_User, time=1, value=0.2}
        ]}}})
    ScaleA=true
    ScaleRGB=true
   )
   /SetVariables(
    Particles.Darken Color:LinearColor=&MakeLinearColorFromVectorAndFloat(
     "Vector (RGB)"=&MakeVectorFromLinearColorRGB(Color=<rgba 0.0083 0.0099 0.02 1>)
     "Float (Alpha)"=&FloatFromCurve(FloatCurve=DataInterface() {propertyValues {Curve {keys [
          {arriveTangent=19.9622, interpMode=RCIM_Cubic, leaveTangent=19.9622, tangentMode=RCTM_User, time=0, value=0}
          {arriveTangentWeight=0.0943, interpMode=RCIM_Cubic, leaveTangentWeight=0.3771, tangentMode=RCTM_User, time=0.1, value=1}
          {arriveTangent=-3.7959, tangentMode=RCTM_User, time=1, value=0}
         ]}}})
    )
    Particles.Darken Size:Vector2f=&Multiply_Vector2D_ByFloat(Float=1.2, Vector2D=$Particles.SpriteSize)
   ) {isSetParameters=true}
   /SetVariables(Particles.Wide Flare Scale:Vector2f=&Vector2DFromCurve(
     Vector2Curve=DataInterface() {propertyValues {
       XCurve {keys [
         {arriveTangent=4.673, interpMode=RCIM_Cubic, leaveTangent=4.673, tangentMode=RCTM_User, time=0, value=0}
         {arriveTangentWeight=0.1536, interpMode=RCIM_Cubic, leaveTangentWeight=0.1152, tangentMode=RCTM_User, time=0.3, value=1}
         {arriveTangent=-3.1914, interpMode=RCIM_Cubic, leaveTangent=-3.1914, tangentMode=RCTM_User, time=0.5, value=0.15}
         {arriveTangent=-2.3891, interpMode=RCIM_Cubic, leaveTangent=-2.3891, tangentMode=RCTM_User, time=0.7, value=0.25}
         {interpMode=RCIM_Cubic, leaveTangent=-1.231, time=1, value=0.2}
        ]}
       YCurve {keys [
         {arriveTangent=10.6903, interpMode=RCIM_Cubic, leaveTangent=10.6903, tangentMode=RCTM_User, time=0, value=0}
         {arriveTangent=0.8, interpMode=RCIM_Cubic, leaveTangent=0.8, time=0.25, value=1}
         {arriveTangent=-1.7663, interpMode=RCIM_Cubic, leaveTangent=-1.7663, tangentMode=RCTM_User, time=0.5, value=0.4}
         {arriveTangent=-4.8806, interpMode=RCIM_Cubic, leaveTangent=-4.8806, tangentMode=RCTM_User, time=0.7, value=1}
         {interpMode=RCIM_Cubic, leaveTangent=-2.2148, time=1, value=0.4}
        ]}
      }}
     ScaleCurve=&Multiply_Vector2D_ByFloat(Float=&MakeFloatFromVector(Vector=$Engine.Owner.Scale, Channel=::NewEnumerator0), Vector2D=<2000 300>)
    )) {isSetParameters=true}
   /SetVariables(Particles.LightColor=&MultiplyLinearColor(Linear Color=$Particles.Color, Scale Factor=<2 0.7 1 50>)) {isSetParameters=true}
  ]
  rendererClasses [&NiagaraSpriteRendererProperties, &NiagaraLightRendererProperties]
  renderers [
   {rendererClass=&NiagaraSpriteRendererProperties}
   {rendererClass=&NiagaraSpriteRendererProperties, rendererIndex=1}
   {rendererClass=&NiagaraSpriteRendererProperties, rendererIndex=2}
   {rendererClass=&NiagaraLightRendererProperties, rendererIndex=3}
  ]
 }
 Horizontal_Shockwave {
  emitterUpdateScript [
   &EmitterState()
   &SpawnBurst_Instantaneous(Use Spawn Probability=false, UseLoopCountLimit=false) {module=SpawnBurst_Instantaneous001}
  ]
  particleSpawnScript [
   &InitializeParticle(
    Color Mode=::NewEnumerator0
    Lifetime Max=0.7
    Lifetime Min=0.5
    Lifetime Mode=::NewEnumerator1
    Mesh Scale Mode=::NewEnumerator3
    Mesh Uniform Scale=&Multiply_Float(A=120, B=1)
    Position Offset=<0 0 100>
    UsePositionOffset=true
    Write Material Random=true
   )
   &CameraOffset(Camera Offset Amount=100)
   &InitialMeshOrientation(Mesh Orientation Mode=::NewEnumerator1, Use Rotation Vector=true)
  ]
  particleUpdateScript [
   &ParticleState()
   &ScaleMeshSize(Scale Factor=&VectorFromFloat(Value=&FloatFromCurve(
      FloatCurve=DataInterface() {propertyValues {Curve {
         defaultValue=1
         keys [
          {interpMode=RCIM_Cubic, tangentMode=RCTM_User, time=0, value=1}
          {arriveTangent=-2.6525, interpMode=RCIM_Cubic, leaveTangent=-2.6525, tangentMode=RCTM_User, time=1, value=0}
         ]
        }}}
      Scale Curve=&Multiply_Float(A=0.25, B=$User.Uniform Scale)
     )))
   &DynamicMaterialParameters(
    Index 0 Param 2=&FloatFromCurve(FloatCurve=DataInterface() {propertyValues {Curve {keys [<at 0 1>,<at 1 0.7>]}}})
    Module.Param0WriteEnabled=true
    Module.Param1WriteEnabled=true
    Module.Param2WriteEnabled=true
    Module.Param3WriteEnabled=true
   )
   &ScaleColor(
    Linear Color Curve=DataInterface() {propertyValues {
      AlphaCurve {keys [<at 0 1>]}
      BlueCurve {keys [<at 0.1342 1>,<at 0.5341 1>]}
      GreenCurve {keys [<at 0.1342 0.0398>,<at 0.5341 0.1016>]}
      RedCurve {keys [<at 0.1342 0.879>,<at 0.5341 0.0321>]}
     }}
    Scale Mode=::NewEnumerator2
   ) {module=ScaleColor001}
   &ScaleColor(
    Scale Alpha=&FloatFromCurve(
     FloatCurve=DataInterface() {propertyValues {Curve {defaultValue=0, keys [{tangentMode=RCTM_User, time=0, value=0}, {tangentMode=RCTM_User, time=1, value=1}]}}}
     Scale Curve=0.15
    )
    ScaleA=true
    ScaleRGB=true
   )
  ]
  rendererClasses [&NiagaraMeshRendererProperties]
  renderers [{rendererClass=&NiagaraMeshRendererProperties, rendererIndex=0}]
 }
 Iris_Flare {
  bEnabled=false
  emitterUpdateScript [
   &EmitterState()
   &SpawnBurst_Instantaneous(Spawn Count=1, Use Spawn Probability=false)
   &SpawnBurst_Instantaneous(Spawn Count=1, Spawn Time=0.3, Use Spawn Probability=false) {module=SpawnBurst_Instantaneous001}
  ]
  particleSpawnScript [
   &InitializeParticle(
    Color Mode=::NewEnumerator0
    Sprite Size Mode=::NewEnumerator3
    Uniform Sprite Size=&Multiply_Float(A=1200, B=$User.Uniform Scale)
    UsePositionOffset=false
    Write Material Random=true
   )
   &CameraOffset(Camera Offset Amount=100)
  ]
  particleUpdateScript [
   &ParticleState()
   &ScaleColor(
    Linear Color Curve=DataInterface() {propertyValues {
      AlphaCurve {keys [<at 0 1>]}
      BlueCurve {keys [<at 0 1>,<at 0.0842 0.4355>,<at 0.3189 0.9859>,<at 1 1>]}
      GreenCurve {keys [<at 0 0.1198>,<at 0.0842 1>,<at 0.3189 1>,<at 1 0.1198>]}
      RedCurve {keys [<at 0 0.1931>,<at 0.0842 0.1936>,<at 0.3189 0.151>,<at 1 0.3545>]}
     }}
    Scale Mode=::NewEnumerator2
   )
   &ScaleColor(
    Scale Alpha=&FloatFromCurve(FloatCurve=DataInterface() {propertyValues {Curve {keys [<at 0 1>,<at 0.1 0.3>,<at 1 0>]}}})
    Scale RGB=&VectorFromFloat(Value=&FloatFromCurve(FloatCurve=DataInterface() {propertyValues {Curve {keys [<at 0 5>,<at 0.1 1>,<at 1 0>]}}}))
    ScaleA=true
    ScaleRGB=true
   ) {module=ScaleColor001}
   &ScaleSpriteSize(Scale Factor=&Vector2DFromFloat(Value=&FloatFromCurve(
      FloatCurve=DataInterface() {propertyValues {Curve {keys [<at 0 0>,<at 0.1 1>,<at 1 2>]}}}
      Scale Curve=&Multiply_Float(A=1, B=$User.Uniform Scale)
     )))
   /SetVariables(
    Particles.RingColor:LinearColor=&MultiplyLinearColor(
     Linear Color=&ColorFromCurve(ColorCurve=DataInterface() {propertyValues {
        AlphaCurve {keys [<at 0 1 RCIM_Cubic>,<at 0.5 0>]}
        BlueCurve {keys [<at 0 2 RCIM_Cubic>,<at 0.5 2 RCIM_Cubic>]}
        GreenCurve {keys [<at 0 0.087 RCIM_Cubic>,<at 0.5 0.0775 RCIM_Cubic>]}
        RedCurve {keys [<at 0 0.5267 RCIM_Cubic>,<at 0.5 0.0417 RCIM_Cubic>]}
       }})
     Scale Factor=&MakeFloat4FromVectorAndFloat(
      "Vector (RGB)"=&VectorFromFloat(Value=&FloatFromCurve(FloatCurve=DataInterface() {propertyValues {Curve {keys [<at 0 1>,<at 0.1 0.5>,<at 0.5 0>]}}}))
      "Float (Alpha)"=0.1
     )
    )
    Particles.RingDynamic:Vector4f=&MakeVector4(X=2, Y=0, Z=1, W=1)
    Particles.RingSize:Vector2f=&Vector2DFromFloat(Value=&Multiply_Float(
      A=1500
      B=&FloatFromCurve(
       FloatCurve=DataInterface() {propertyValues {Curve {
          defaultValue=0
          keys [
           {arriveTangent=4.1785, interpMode=RCIM_Cubic, leaveTangent=4.1785, tangentMode=RCTM_User, time=0, value=0}
           {arriveTangent=1.1179, interpMode=RCIM_Cubic, leaveTangent=1.1179, tangentMode=RCTM_User, time=0.2042, value=0.6228}
           {arriveTangent=0.0224, interpMode=RCIM_Cubic, leaveTangent=0.0224, tangentMode=RCTM_User, time=1, value=1}
          ]
         }}}
       Scale Curve=&Multiply_Float(A=1, B=$User.Uniform Scale)
      )
     ))
   ) {isSetParameters=true}
   /SetVariables(
    Particles.DarkColor:LinearColor=&MultiplyLinearColor(
     Linear Color=&ColorFromCurve(ColorCurve=DataInterface() {propertyValues {
        AlphaCurve {keys [<at 0 1 RCIM_Cubic>,<at 1 0>]}
        BlueCurve {keys [<at 0 0.0156 RCIM_Cubic>, {arriveTangent=-0.0156, interpMode=RCIM_Cubic, leaveTangent=-0.0156, time=0.3949, value=0.0052}, <at 1 0>]}
        GreenCurve {keys [<at 0 0.0007 RCIM_Cubic>, {arriveTangent=-0.0007, interpMode=RCIM_Cubic, leaveTangent=-0.0007, time=0.3949, value=0.0011}, <at 1 0>]}
        RedCurve {keys [<at 0 0.005 RCIM_Cubic>, {arriveTangent=-0.005, interpMode=RCIM_Cubic, leaveTangent=-0.005, time=0.3949, value=0.0005}, <at 1 0>]}
       }})
     Scale Factor=&MakeFloat4FromVectorAndFloat("Vector (RGB)"=<1 1 1>, "Float (Alpha)"=0.5)
    )
    Particles.DarkSize:Vector2f=&Vector2DFromFloat(Value=&FloatFromCurve(
      FloatCurve=DataInterface() {propertyValues {Curve {keys [<at 0 0>,<at 0.1 1>,<at 1 2>]}}}
      Scale Curve=&Multiply_Float(A=1200, B=$User.Uniform Scale)
     ))
   ) {isSetParameters=true}
  ]
  rendererClasses [&NiagaraSpriteRendererProperties]
  renderers [
   {rendererClass=&NiagaraSpriteRendererProperties}
   {rendererClass=&NiagaraSpriteRendererProperties, rendererIndex=1}
  ]
 }
 NE_Energy_WAve_In {
  emitterUpdateScript [
   &EmitterState()
   &SpawnBurst_Instantaneous(Spawn Group=1, Use Spawn Probability=false, UseLoopCountLimit=false) {module=SpawnBurst_Instantaneous003}
   &SpawnBurst_Instantaneous(Spawn Time=1.3, Use Spawn Probability=false, UseLoopCountLimit=false)
  ]
  particleSpawnScript [
   &InitializeParticle(
    Color Mode=::NewEnumerator0
    Lifetime=&FloatFromCurve(FloatCurve=DataInterface() {propertyValues {Curve {
        defaultValue=0.7647
        keys [
         {arriveTangent=-2.0749, interpMode=RCIM_Cubic, leaveTangent=-2.0749, tangentMode=RCTM_User, time=0, value=1}
         {interpMode=RCIM_Cubic, tangentMode=RCTM_User, time=1, value=0.3}
        ]
       }}})
    Mesh Scale=<5 10 10>
    Mesh Scale Mode=::NewEnumerator4
    UsePositionOffset=false
    Write Material Random=true
   )
   &CameraOffset(Camera Offset Amount=100)
  ]
  particleUpdateScript [
   &ParticleState()
   &ScaleMeshSize(Scale Factor=<2 0.5 0.5>)
   &ScaleMeshSize(Scale Factor=&VectorFromFloat(Value=&FloatFromCurve(
      FloatCurve=DataInterface() {propertyValues {Curve {
         defaultValue=1
         keys [
          {arriveTangent=0.8158, interpMode=RCIM_Cubic, leaveTangent=0.8158, tangentMode=RCTM_User, time=0, value=1}
          {arriveTangent=-3.2857, interpMode=RCIM_Cubic, leaveTangent=-3.2857, tangentMode=RCTM_User, time=1, value=0}
         ]
        }}}
      Scale Curve=&Multiply_Float(A=0.25, B=$User.Uniform Scale)
     ))) {module=ScaleMeshSize001}
   &ScaleColor(
    Linear Color Curve=DataInterface() {propertyValues {
      AlphaCurve {keys [<at 0 1>]}
      BlueCurve {keys [<at 0.3089 1>,<at 0.5341 1>]}
      GreenCurve {keys [<at 0.3089 0.1387>,<at 0.5341 0.1016>]}
      RedCurve {keys [<at 0.3089 0.7917>,<at 0.5341 0.0321>]}
     }}
    Scale Mode=::NewEnumerator2
   ) {enabled=false}
   &ScaleColor(
    Scale Alpha=&FloatFromCurve(
     FloatCurve=DataInterface() {propertyValues {Curve {keys [
         {interpMode=RCIM_Cubic, leaveTangentWeight=0.7975, tangentMode=RCTM_User, time=0, value=0}
         {arriveTangentWeight=0.4526, interpMode=RCIM_Cubic, leaveTangentWeight=0.4526, tangentMode=RCTM_User, time=0.5, value=1}
         {arriveTangent=-5.9523, interpMode=RCIM_Cubic, leaveTangent=-5.9523, tangentMode=RCTM_User, time=1, value=0}
        ]}}}
     Scale Curve=0.3
    )
    Scale RGB=&VectorFromFloat(Value=&FloatFromCurve(
      FloatCurve=DataInterface() {propertyValues {Curve {
         defaultValue=0
         keys [
          {arriveTangent=4.1785, interpMode=RCIM_Cubic, leaveTangent=4.1785, tangentMode=RCTM_User, time=0, value=0}
          {arriveTangent=1.1179, interpMode=RCIM_Cubic, leaveTangent=1.1179, tangentMode=RCTM_User, time=0.2042, value=0.6228}
          {arriveTangent=0.0224, interpMode=RCIM_Cubic, leaveTangent=0.0224, tangentMode=RCTM_User, time=1, value=1}
         ]
        }}}
      Scale Curve=&FNTimeOfDayScalar(Night Scale=0.2, Morning Scale=0.3, Evening Scale=0.3)
     ))
    ScaleA=true
    ScaleRGB=false
   ) {module=ScaleColor001}
   &DynamicMaterialParameters(
    Index 0 Param 1=&FloatFromCurve(FloatCurve=DataInterface() {propertyValues {Curve {
        defaultValue=0
        keys [
         {arriveTangent=4.1785, interpMode=RCIM_Cubic, leaveTangent=4.1785, tangentMode=RCTM_User, time=0, value=0}
         {arriveTangent=1.1179, interpMode=RCIM_Cubic, leaveTangent=1.1179, tangentMode=RCTM_User, time=0.2042, value=0.6228}
         {arriveTangent=0.0224, interpMode=RCIM_Cubic, leaveTangent=0.0224, tangentMode=RCTM_User, time=1, value=1}
        ]
       }}})
    Index 0 Param 2=&FloatFromCurve(FloatCurve=DataInterface() {propertyValues {Curve {
        defaultValue=1
        keys [{interpMode=RCIM_Cubic, tangentMode=RCTM_User, time=0, value=-0.7}, {interpMode=RCIM_Cubic, tangentMode=RCTM_User, time=1, value=0}]
       }}})
    Index 0 Param 3=&FloatFromCurve(
     FloatCurve=DataInterface() {propertyValues {Curve {
        defaultValue=0
        keys [{interpMode=RCIM_Cubic, tangentMode=RCTM_User, time=0, value=1}, {interpMode=RCIM_Cubic, tangentMode=RCTM_User, time=1, value=-1.5217}]
       }}}
     Scale Curve=0.3
    )
    Index 0 Param 4=&RandomRangeFloat(Minimum=1, Maximum=1)
    Index 1 Param 1=&FloatFromCurve(FloatCurve=DataInterface() {propertyValues {Curve {keys [
         {arriveTangent=-0.0851, interpMode=RCIM_Cubic, leaveTangent=-0.0851, tangentMode=RCTM_User, time=0, value=0}
         {arriveTangent=1.1179, leaveTangent=1.1179, time=0.2277, value=1}
         {arriveTangent=0.0224, interpMode=RCIM_Cubic, leaveTangent=0.0224, tangentMode=RCTM_User, time=1, value=1}
        ]}}})
    Index 2 Param 1=&FloatFromCurve(FloatCurve=DataInterface() {propertyValues {Curve {keys [
         {
          arriveTangent=-0.0147
          arriveTangentWeight=0.7781
          interpMode=RCIM_Cubic
          leaveTangent=-0.0147
          leaveTangentWeight=0.7781
          tangentMode=RCTM_User
          tangentWeightMode=RCTWM_WeightedBoth
          time=0
          value=0.2
         }
         {arriveTangent=-0.3563, interpMode=RCIM_Cubic, leaveTangent=-0.3563, tangentMode=RCTM_User, time=1, value=2}
        ]}}})
    Module.Index0Float0Write=true
    Module.Index0Float1Write=true
    Module.Index0Float2Write=true
    Module.Index0Float3Write=true
    Module.Index1Float0Write=true
    Module.Index1Float1Write=true
    Module.Index1Float2Write=true
    Module.Index1Float3Write=true
    Module.Index2Float0Write=true
    Module.Index2Float1Write=true
    Module.Index2Float2Write=true
    Module.Index2Float3Write=true
    Write Parameter Index 1=::NewEnumerator1
    Write Parameter Index 2=::NewEnumerator1
   )
   &MeshRotationForce(Rotation=<11 0 0>) {enabled=false}
   &SolveForcesAndVelocity(Clamp Velocity=false, Limit Acceleration=false)
  ]
  rendererClasses [&NiagaraMeshRendererProperties]
  renderers [{rendererClass=&NiagaraMeshRendererProperties, rendererIndex=0}]
 }
 NS_SoftAirflow {
  emitterUpdateScript [
   &EmitterState()
   &SpawnBurst_Instantaneous(Spawn Count=3, Use Spawn Probability=false, UseLoopCountLimit=false) {module=SpawnBurst_Instantaneous001}
  ]
  particleSpawnScript [
   &InitializeParticle(
    Lifetime Max=1.2
    Lifetime Min=0.7
    Lifetime Mode=::NewEnumerator1
    Mesh Scale Mode=::NewEnumerator1
    Mesh Uniform Scale Max=40
    Mesh Uniform Scale Min=20
    Position Offset=&AddVectorToPosition(Position=<0 0 0>, Vector=<0 0 0>)
    UsePositionOffset=false
    Write Material Random=true
   )
   &InitialMeshOrientation(
    Mesh Orientation Mode=::NewEnumerator1
    Rotation=&RandomRangeVector(Minimum=<-0.1 -0.1 0>, Maximum=<0.1 0.1 1>)
    Use Rotation Vector=true
   )
  ]
  particleUpdateScript [
   &ParticleState()
   &DynamicMaterialParameters(
    Index 0 Param 1=&FloatFromCurve(FloatCurve=DataInterface() {propertyValues {Curve {keys [{interpMode=RCIM_Cubic, tangentMode=RCTM_User, time=0.5, value=0}, {interpMode=RCIM_Cubic, tangentMode=RCTM_User, time=1, value=1}]}}})
    Index 0 Param 3=1
    Module.Param0WriteEnabled=true
    Module.Param1WriteEnabled=true
    Module.Param2WriteEnabled=true
    Module.Param3WriteEnabled=true
   )
   &ScaleMeshSize(Scale Factor=&Multiply_VectorByFloat(
     Vector=<1 1 1.5>
     Float=&FloatFromCurve(
      FloatCurve=DataInterface() {propertyValues {Curve {defaultValue=0, keys [{tangentMode=RCTM_User, time=0, value=1}, {tangentMode=RCTM_User, time=1, value=0}]}}}
      Scale Curve=&Multiply_Float(A=0.8, B=$User.Uniform Scale)
     )
    ))
   &UpdateMeshOrientation(
    Rotation Rate=&FloatFromCurve(FloatCurve=DataInterface() {propertyValues {Curve {keys [{arriveTangent=-2.0983, interpMode=RCIM_Cubic, leaveTangent=-2.0983, tangentMode=RCTM_User, time=0, value=1}, <at 1 0>]}}})
    Rotation Vector=<0 0 1>
   )
   &ScaleColor(
    Scale Alpha=&FloatFromCurve(
     FloatCurve=DataInterface() {propertyValues {Curve {keys [
         {arriveTangent=8.2356, interpMode=RCIM_Cubic, leaveTangent=8.2356, tangentMode=RCTM_User, time=0, value=0}
         {arriveTangent=0.1969, interpMode=RCIM_Cubic, leaveTangent=0.1969, tangentMode=RCTM_User, time=0.22, value=1}
         {arriveTangent=-1.2821, interpMode=RCIM_Cubic, leaveTangent=-1.2821, time=0.5, value=0.6}
         {arriveTangent=-0.7636, interpMode=RCIM_Cubic, leaveTangent=-0.7636, tangentMode=RCTM_User, time=1, value=0}
        ]}}}
     Scale Curve=0.3
    )
    Scale RGB=&MakeVectorFromLinearColorRGB(Color=&ScaleLinearColorByCurve(
      InitialValue=<rgba 1 1 1 1>
      ScaleCurve=DataInterface() {propertyValues {
        AlphaCurve {keys []}
        BlueCurve {keys [<at 0.1749 1>,<at 0.5228 5>]}
        GreenCurve {keys [<at 0.1749 0.0417>,<at 0.5228 0.4678>]}
        RedCurve {keys [<at 0.1749 0.8289>,<at 0.5228 0>]}
       }}
      CurveIndex=$Particles.NormalizedAge
     ))
    ScaleA=true
    ScaleRGB=true
   )
  ]
  rendererClasses [&NiagaraMeshRendererProperties]
  renderers [{rendererClass=&NiagaraMeshRendererProperties, rendererIndex=0}]
 }
 Tether {
  bEnabled=false
  emitterUpdateScript [
   &EmitterState(
    Inactive Response=::NewEnumerator1
    Life Cycle Mode=::NewEnumerator1
    Loop Behavior=::NewEnumerator1
    Loop Duration Mode=::NewEnumerator1
    UseLoopDelay=false
   )
   &SpawnBurst_Instantaneous(Spawn Count=24, Use Spawn Probability=false, UseLoopCountLimit=false)
  ]
  particleSpawnScript [
   &PartitionParticles(
    Normalized Index In Partition Includes One=true
    Particles Per Partition=12
    Partition Mode:ENiagara_PartitionMode_v2=::NewEnumerator1
    Write Partition Index to Ribbon Link Order:ENiagara_AttributeSamplingApplyOutput=::NewEnumerator0
    Write Partition to Ribbon ID=::NewEnumerator0
   )
   &InitializeParticle(
    Color=&LerpLinearColors(
     StartColor=<rgba 1 0.1208 0.6171 1>
     EndColor=<rgba 0.148 0.0378 0.5 0.75>
     LerpFactor=&FloatFromCurve(
      FloatCurve=DataInterface() {propertyValues {Curve {
         defaultValue=1
         keys [
          {arriveTangent=26.9494, interpMode=RCIM_Cubic, time=0, value=0}
          {arriveTangentWeight=0.1, interpMode=RCIM_Cubic, leaveTangentWeight=0.1333, tangentMode=RCTM_User, time=0.2, value=1}
          {arriveTangent=-2.6193, interpMode=RCIM_Cubic, leaveTangent=-2.6193, tangentMode=RCTM_User, time=0.8242, value=0.5921}
          <at 1 0 RCIM_Cubic>
         ]
        }}}
      CurveIndex=$Particles.PartitionParticles.Normalized Index In Partition
     )
    )
    Lifetime=2
    Ribbon Facing Mode=::NewEnumerator1
    Ribbon Width=&FloatFromCurve(
     FloatCurve=DataInterface() {propertyValues {Curve {keys [<at 0.25 1 RCIM_Cubic>, {arriveTangent=-1.4976, tangentMode=RCTM_User, time=1, value=0.4}]}}}
     CurveIndex=$Particles.PartitionParticles.Normalized Index In Partition
     Scale Curve=400
    )
    Ribbon Width Mode=::NewEnumerator1
    UsePositionOffset=false
    Write Material Random=true
   )
   &DynamicMaterialParameters(
    Index 0 Param 1=&FloatFromCurve(
     FloatCurve=DataInterface() {propertyValues {Curve {
        defaultValue=1
        keys [
         {arriveTangent=-10.2466, interpMode=RCIM_Cubic, leaveTangent=-10.2466, leaveTangentWeight=0.1667, tangentMode=RCTM_User, time=0, value=1}
         {arriveTangentWeight=0.1067, interpMode=RCIM_Cubic, leaveTangentWeight=0.3202, tangentMode=RCTM_User, time=0.2, value=0}
         {arriveTangentWeight=0.3202, interpMode=RCIM_Cubic, leaveTangentWeight=0.1067, tangentMode=RCTM_User, time=0.8, value=0}
         {tangentMode=RCTM_User, time=1, value=1}
        ]
       }}}
     CurveIndex=$Particles.PartitionParticles.Normalized Index In Partition
     Scale Curve=0.5
    )
    Index 0 Param 2=&MaskFloat(Mask Source=0, Mask Parameter=?, Mask=$Particles.PartitionParticles.Partition, Masked Float=10, Passthrough Float=5)
    Module.Index0Float0Write=true
    Module.Index0Float1Write=true
    Module.Index0Float2Write=false
    Module.Index0Float3Write=false
   )
  ]
  particleUpdateScript [
   &ParticleState(Kill Particles When Lifetime Has Elapsed=false)
   /SetVariables(Particles.Position=&AddVectorToPosition(
     Position=&LerpPosition(
      A=&TransformPosition(Position=$User.TreePosition, Source Space=::World, Destination Space=::Simulation)
      B=<1500 100 -1600>
      Alpha=$Particles.PartitionParticles.Normalized Index In Partition
     )
     Vector=&MaskVector(
      Mask Source=1
      Mask Parameter=$Particles.PartitionParticles.Partition
      Mask=0
      Masked Vector=&VectorFromCurve(
       VectorCurve=DataInterface() {propertyValues {
         XCurve {keys [
           {arriveTangent=-3.0139, interpMode=RCIM_Cubic, leaveTangent=-3.0139, tangentMode=RCTM_User, time=0, value=0}
           {arriveTangent=0.6894, interpMode=RCIM_Cubic, leaveTangent=0.6894, time=0.1, value=-0.15}
           {arriveTangent=-0.7546, interpMode=RCIM_Cubic, leaveTangent=-0.7546, tangentMode=RCTM_User, time=0.2699, value=0.1861}
           <at 1 0>
          ]}
         YCurve {keys [
           {arriveTangent=28.8057, interpMode=RCIM_Cubic, leaveTangent=28.8057, tangentMode=RCTM_User, time=0, value=0}
           {arriveTangent=-11.1916, interpMode=RCIM_Cubic, leaveTangent=-11.1916, tangentMode=RCTM_User, time=0.0706, value=0.7253}
           {arriveTangent=0.352, interpMode=RCIM_Cubic, leaveTangent=0.352, tangentMode=RCTM_User, time=0.2969, value=-1}
           {arriveTangent=0.531, interpMode=RCIM_Cubic, leaveTangent=0.531, tangentMode=RCTM_User, time=0.7626, value=0.4248}
           <at 1 0 RCIM_Cubic>
          ]}
         ZCurve {keys [<at 0 0 RCIM_Cubic>, <at 0.6384 -1 RCIM_Cubic>, {interpMode=RCIM_Cubic, leaveTangent=0.5706, time=1, value=0}]}
        }}
       CurveIndex=$Particles.PartitionParticles.Normalized Index In Partition
       Scale Curve=<500 500 500>
      )
      Passthrough Vector=&VectorFromCurve(
       VectorCurve=DataInterface() {propertyValues {
         XCurve {keys [
           {arriveTangent=-3.0139, interpMode=RCIM_Cubic, leaveTangent=-3.0139, tangentMode=RCTM_User, time=0, value=0}
           {arriveTangent=0.9424, interpMode=RCIM_Cubic, leaveTangent=0.9424, tangentMode=RCTM_User, time=0.1535, value=-0.0617}
           {arriveTangent=-0.7546, interpMode=RCIM_Cubic, leaveTangent=-0.7546, tangentMode=RCTM_User, time=0.5393, value=0.0203}
           {interpMode=RCIM_Cubic, leaveTangent=0.2713, time=1, value=0}
          ]}
         YCurve {keys [
           {arriveTangent=-17.2386, interpMode=RCIM_Cubic, leaveTangent=-17.2386, tangentMode=RCTM_User, time=0, value=0}
           {arriveTangent=0.1276, interpMode=RCIM_Cubic, leaveTangent=0.1276, tangentMode=RCTM_User, time=0.0604, value=-0.3152}
           {arriveTangent=-1.2086, interpMode=RCIM_Cubic, leaveTangent=-1.2086, tangentMode=RCTM_User, time=0.4121, value=0.5}
           {arriveTangent=-0.8505, interpMode=RCIM_Cubic, leaveTangent=-0.8505, time=0.6967, value=-0.1881}
           {arriveTangent=-1.1853, interpMode=RCIM_Cubic, leaveTangent=-1.1853, tangentMode=RCTM_User, time=1, value=0}
          ]}
         ZCurve {keys [
           <at 0 0 RCIM_Cubic>
           {arriveTangent=-2.317, interpMode=RCIM_Cubic, leaveTangent=-2.317, tangentMode=RCTM_User, time=0.1869, value=-0.6658}
           {arriveTangent=0.1607, interpMode=RCIM_Cubic, leaveTangent=0.1607, tangentMode=RCTM_User, time=0.6453, value=-1}
           {arriveTangent=5.1933, interpMode=RCIM_Cubic, leaveTangent=5.1933, tangentMode=RCTM_User, time=1, value=0}
          ]}
        }}
       CurveIndex=$Particles.PartitionParticles.Normalized Index In Partition
       Scale Curve=<500 500 400>
      )
     )
    )) {isSetParameters=true}
  ]
  rendererClasses [&NiagaraSpriteRendererProperties, &NiagaraRibbonRendererProperties]
  renderers [
   {rendererClass=&NiagaraSpriteRendererProperties}
   {rendererClass=&NiagaraRibbonRendererProperties, rendererIndex=1}
  ]
 }
}
summary {
 systemName=NS_LimeRock_Treehouse_Energy_Suppressed
 emitters [
  {
   bEnabled=false
   emitterName=Flare_Complex
   rendererClasses [&NiagaraSpriteRendererProperties, &NiagaraLightRendererProperties]
  }
  {bEnabled=false, emitterName=ElectricSpirtes, rendererClasses [&NiagaraSpriteRendererProperties]}
  {bEnabled=false, emitterName=Iris_Flare, rendererClasses [&NiagaraSpriteRendererProperties]}
  {emitterName=Horizontal_Shockwave, rendererClasses [&NiagaraMeshRendererProperties]}
  {emitterName=NS_SoftAirflow, rendererClasses [&NiagaraMeshRendererProperties]}
  {emitterName=NE_Energy_WAve_In, rendererClasses [&NiagaraMeshRendererProperties]}
  {bEnabled=false, emitterName=Energy_Offset001, rendererClasses [&NiagaraMeshRendererProperties]}
  {bEnabled=false, emitterName=Burst_Emitter, rendererClasses [&NiagaraMeshRendererProperties]}
  {
   bEnabled=false
   emitterName=Tether
   rendererClasses [&NiagaraSpriteRendererProperties, &NiagaraRibbonRendererProperties]
  }
 ]
 userVariables [
  User.TreePosition {type=NiagaraPosition, description="", defaultValue {}}
  User.Uniform Scale {type=NiagaraFloat, defaultValue=2, description=""}
 ]
}
