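# Niagara system for the LimeRock treehouse bubble, a purple reality-bubble effect with a glowing mesh shell and light.
# Pulsing flare shockwaves, mesh lightning on a sphere, vortex-swirled motes and GPU shimmer sprites layer around it.
# System open and destabilize values drive mote color, spread and turbulence plus bubble material; User parameters control radius, lightning rate and fade-out.
# The Shimmers and MeshRibbon emitters are disabled.
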NiagaraSystem /LimeRock/Assets/FX/Treehouse_Bubble/NS_LimeRock_Treehouse_FX
dependencies {
 usedDataInterfaces [&NiagaraDataInterfaceCurve, &NiagaraDataInterfaceColorCurve, &NiagaraDataInterfaceCamera]
 usedDynamicInputs [&ReturnCameraVector]
 usedModules [
  &SystemState
  &self:SystemSpawnScript.SystemScriptSource.SystemScriptGraph.NiagaraNodeAssignment_0.SetVariables_#1
  &self:SystemSpawnScript.SystemScriptSource.SystemScriptGraph.NiagaraNodeAssignment_1.SetVariables_#2
  &self:CalculatePortalOpenState
 ]
}
emitters {
 FlareLightShockwave {
  emitterSpawnScript [/SetVariables(Emitter.DepthFade=100) {isSetParameters=true}]
  emitterUpdateScript [
   &EmitterState(Life Cycle Mode=::NewEnumerator1, UseLoopDelay=false)
   &SpawnBurst_Instantaneous(Use Spawn Probability=false, UseLoopCountLimit=false)
  ]
  particleSpawnScript [
   &InitializeParticle(
    Color=<rgba 0.4325 0.1181 1 1>
    Lifetime=5.5
    Mesh Scale Mode=::NewEnumerator3
    Mesh Uniform Scale=&Divide_Float(A=$User.Radius, B=150)
    Position Offset=<15 0 0>
    Sprite Size Mode=::NewEnumerator3
    Uniform Sprite Size=2500
    UsePositionOffset=true
    Write Material Random=true
   )
   &ShapeLocation(MaskSpawnGroup=false, Offset=$System.Offset, Offset Mode=::NewEnumerator0, Override Seed=false, Sphere Radius=0, UseOwnerScale=true)
  ]
  particleUpdateScript [
   &ParticleState()
   &CameraOffset(Camera Offset Amount=&FloatFromCurve(
     FloatCurve=DataInterface() {propertyValues {Curve {keys [
         {arriveTangent=4.0899, interpMode=RCIM_Cubic, leaveTangent=4.0899, tangentMode=RCTM_User, time=0, value=0.5}
         {arriveTangent=-3.6426, interpMode=RCIM_Cubic, leaveTangent=-3.6426, tangentMode=RCTM_User, time=0.2, value=1}
         {arriveTangent=5.165, interpMode=RCIM_Cubic, leaveTangent=5.165, tangentMode=RCTM_User, time=0.4, value=0.3}
         <at 0.75 0>
        ]}}}
     Scale Curve=200
    ))
   &ScaleMeshSize(Scale Factor=&VectorFromFloat(Value=10))
   &DynamicMaterialParameters(
    Index 0 Param 1=1
    Index 0 Param 2=&Lerp_Float(
     A=1
     B=0.3
     Alpha=&FloatFromCurve(FloatCurve=DataInterface() {propertyValues {Curve {keys [
          {arriveTangent=5.1433, interpMode=RCIM_Cubic, time=0, value=0}
          {arriveTangent=4.5726, interpMode=RCIM_Cubic, leaveTangent=4.5726, tangentMode=RCTM_User, time=0.25, value=0.7}
          {arriveTangent=-6.4921, interpMode=RCIM_Cubic, leaveTangent=2.9849, tangentMode=RCTM_Break, time=0.4, value=0.35}
          {arriveTangent=0.6574, interpMode=RCIM_Cubic, leaveTangent=0.6574, tangentMode=RCTM_User, time=1, value=1}
         ]}}})
    )
    Index 3 Param 1=&Multiply_Float(A=&MakeFloatFromVector(Vector=$Particles.Scale, Channel=::NewEnumerator0), B=90)
    Index 3 Param 2=&FloatFromCurve(
     FloatCurve=DataInterface() {propertyValues {Curve {keys [
         {arriveTangent=2.1437, interpMode=RCIM_Cubic, leaveTangent=2.1437, tangentMode=RCTM_User, time=0, value=0}
         {arriveTangent=1.3259, interpMode=RCIM_Cubic, leaveTangent=1.3259, tangentMode=RCTM_User, time=0.6, value=0.5}
         {interpMode=RCIM_Cubic, leaveTangent=0.2859, time=1, value=1}
        ]}}}
     Scale Curve=0
    )
    Module.Param0WriteEnabled=true
    Module.Param0WriteEnabled3=true
    Module.Param1WriteEnabled=true
    Module.Param1WriteEnabled3=true
    Module.Param2WriteEnabled=false
    Module.Param2WriteEnabled3=false
    Module.Param3WriteEnabled=false
    Module.Param3WriteEnabled3=false
    Write Parameter Index 3=true
   )
   &ScaleSpriteSize(
    Uniform Curve Scale=5
    Uniform Curve Sprite Scale=DataInterface() {propertyValues {Curve {keys [
        {arriveTangent=-6.1148, interpMode=RCIM_Cubic, leaveTangent=-6.1148, tangentMode=RCTM_User, time=0, value=1}
        {arriveTangent=-0.4947, interpMode=RCIM_Cubic, leaveTangent=8.7488, tangentMode=RCTM_Break, time=0.4, value=0.3}
        {
         arriveTangent=-0.2276
         arriveTangentWeight=0.0565
         interpMode=RCIM_Cubic
         leaveTangent=-0.2276
         leaveTangentWeight=0.1797
         tangentMode=RCTM_User
         time=0.5
         value=0.9
        }
        {interpMode=RCIM_Cubic, leaveTangent=-0.9894, time=0.7, value=0.5}
       ]}}}
   )
   &ScaleColor(
    Scale Alpha=&FloatFromCurve(FloatCurve=DataInterface() {propertyValues {Curve {keys [
         {arriveTangent=18.2404, interpMode=RCIM_Cubic, leaveTangent=18.2404, tangentMode=RCTM_User, time=0, value=1}
         {arriveTangent=-1.4815, interpMode=RCIM_Cubic, leaveTangent=-1.4815, time=0.12, value=2}
         {
          arriveTangent=-4.2653
          arriveTangentWeight=0.0565
          interpMode=RCIM_Cubic
          leaveTangent=-4.2653
          leaveTangentWeight=0.1797
          tangentMode=RCTM_User
          time=0.27
          value=0.6
         }
         {arriveTangent=-11.3751, interpMode=RCIM_Cubic, leaveTangent=-11.3751, tangentMode=RCTM_User, time=0.44, value=2}
         {interpMode=RCIM_Cubic, tangentMode=RCTM_User, time=1, value=0}
        ]}}})
    ScaleA=true
    ScaleRGB=true
   )
   /SetVariables(Particles.WaveColor:LinearColor=&MakeLinearColorFromVectorAndFloat(
     "Vector (RGB)"=&Lerp_Vector(
      A=&MakeVectorFromLinearColorRGB(Color=<rgba 0.7829 0 1 1>)
      B=&MakeVectorFromLinearColorRGB(Color=<rgba 0.2363 0.0764 1 1>)
      Alpha=&FloatFromCurve(FloatCurve=DataInterface() {propertyValues {Curve {
          defaultValue=1
          keys [
           {interpMode=RCIM_Cubic, tangentMode=RCTM_User, time=0, value=0}
           {arriveTangent=1.0171, interpMode=RCIM_Cubic, leaveTangent=1.0171, time=0.0954, value=0.8302}
           {interpMode=RCIM_Cubic, tangentMode=RCTM_User, time=0.8483, value=0.8627}
           {tangentMode=RCTM_User, time=1, value=0}
          ]
         }}})
     )
     "Float (Alpha)"=&FloatFromCurve(
      FloatCurve=DataInterface() {propertyValues {Curve {
         defaultValue=1
         keys [
          {interpMode=RCIM_Cubic, tangentMode=RCTM_User, time=0, value=0}
          {interpMode=RCIM_Cubic, tangentMode=RCTM_User, time=0.0967, value=0.9353}
          {arriveTangent=-1.0354, interpMode=RCIM_Cubic, leaveTangent=-1.0354, time=0.9048, value=0.8647}
          {tangentMode=RCTM_User, time=1, value=0}
         ]
        }}}
      Scale Curve=&Lerp_Float(A=0, B=0.5, Alpha=$User.Fade out)
     )
    )) {isSetParameters=true}
   /SetVariables(Particles.LightColor=&MultiplyLinearColor(Linear Color=$Particles.Color, Scale Factor=&MakeVector4(X=10, Y=10, Z=10, W=15))) {isSetParameters=true}
   /SetVariables(Particles.LightRadius=&FloatFromCurve(FloatCurve=DataInterface() {propertyValues {Curve {keys [<at 0.2 0.35 RCIM_Cubic>,<at 0.5 1>]}}}, Scale Curve=1000)) {isSetParameters=true}
  ]
  rendererClasses [
   &NiagaraSpriteRendererProperties
   &NiagaraLightRendererProperties
   &NiagaraMeshRendererProperties
  ]
  renderers [
   {rendererClass=&NiagaraSpriteRendererProperties}
   {rendererClass=&NiagaraLightRendererProperties, rendererIndex=1}
   {rendererClass=&NiagaraMeshRendererProperties, rendererIndex=2}
  ]
 }
 Light001 {
  emitterUpdateScript [
   &EmitterState(
    Inactive Response=::NewEnumerator1
    Life Cycle Mode=::NewEnumerator1
    Loop Behavior=::NewEnumerator1
    Loop Delay=2
    Loop Duration Mode=::NewEnumerator1
    UseLoopDelay=false
   )
   &SpawnBurst_Instantaneous(Spawn Count=1, Use Spawn Probability=false)
  ]
  particleSpawnScript [&InitializeParticle(Color=<rgba 2 0.227 1.1086 1>, Position Offset=<0 0 500>, UsePositionOffset=true, Write Material Random=true) {module=InitializeParticle001}]
  particleUpdateScript [
   &ParticleState(Kill Particles When Lifetime Has Elapsed=false)
   &ScaleSpriteSize(Uniform Curve Sprite Scale=DataInterface() {propertyValues {Curve {keys [
        {interpMode=RCIM_Cubic, tangentMode=RCTM_User, time=0, value=0.3}
        {arriveTangent=0.7525, interpMode=RCIM_Cubic, leaveTangent=0.7525, tangentMode=RCTM_User, time=0.8, value=0.45}
        {interpMode=RCIM_Cubic, tangentMode=RCTM_User, time=0.9, value=1}
       ]}}})
   &ScaleColor(
    Scale Alpha=&FloatFromCurve(FloatCurve=DataInterface() {propertyValues {Curve {keys [
         <at 0 0 RCIM_Cubic>
         {arriveTangent=0.289, interpMode=RCIM_Cubic, leaveTangent=0.289, time=0.4918, value=0.248}
         {arriveTangent=3.4072, interpMode=RCIM_Cubic, leaveTangent=3.4072, time=0.5041, value=0.1457}
         {arriveTangent=-2.7103, interpMode=RCIM_Cubic, leaveTangent=-2.7103, time=0.5232, value=0.3549}
         {arriveTangent=2.94, interpMode=RCIM_Cubic, leaveTangent=2.94, time=0.5374, value=0.0554}
         {arriveTangent=0.8123, interpMode=RCIM_Cubic, leaveTangent=0.8123, time=0.5612, value=0.4666}
         {arriveTangent=3.9277, interpMode=RCIM_Cubic, leaveTangent=3.9277, time=0.5922, value=0.0999}
         {arriveTangent=3.855, interpMode=RCIM_Cubic, leaveTangent=3.855, time=0.6114, value=0.6636}
         {arriveTangent=-6.193, interpMode=RCIM_Cubic, leaveTangent=-6.193, time=0.6303, value=0.2469}
         {arriveTangent=1.1094, interpMode=RCIM_Cubic, leaveTangent=1.1094, time=0.6459, value=0.4498}
         {arriveTangent=3.878, interpMode=RCIM_Cubic, leaveTangent=3.878, time=0.6893, value=0.3124}
         {arriveTangent=0.3841, interpMode=RCIM_Cubic, leaveTangent=0.3841, time=0.6979, value=0.6518}
         {arriveTangent=1.6385, interpMode=RCIM_Cubic, leaveTangent=1.6385, time=0.7238, value=0.3256}
         {arriveTangent=1.2094, interpMode=RCIM_Cubic, leaveTangent=1.2094, time=0.7388, value=0.7187}
         {arriveTangent=5.1593, interpMode=RCIM_Cubic, leaveTangent=5.1593, time=0.769, value=0.3803}
         {arriveTangent=2.2949, interpMode=RCIM_Cubic, leaveTangent=2.2949, time=0.7839, value=0.9515}
         {arriveTangent=-4.6065, interpMode=RCIM_Cubic, leaveTangent=-4.6065, time=0.815, value=0.4859}
         {arriveTangent=9.5468, interpMode=RCIM_Cubic, leaveTangent=9.5468, time=0.8665, value=0.5711}
         {arriveTangent=0.1948, interpMode=RCIM_Cubic, leaveTangent=0.1948, time=0.8707, value=1.017}
         {arriveTangent=-7.1532, interpMode=RCIM_Cubic, leaveTangent=-7.1532, time=0.8934, value=0.5764}
         {arriveTangent=15.7037, interpMode=RCIM_Cubic, leaveTangent=15.7037, time=0.925, value=0.6286}
         {arriveTangent=-2.0814, interpMode=RCIM_Cubic, leaveTangent=-2.0814, time=0.9294, value=1.1414}
         {arriveTangent=-16.7837, interpMode=RCIM_Cubic, leaveTangent=-16.7837, time=0.9479, value=0.5809}
         {arriveTangent=31.4448, interpMode=RCIM_Cubic, leaveTangent=31.4448, time=0.9612, value=0.6074}
         {arriveTangent=-15.6706, interpMode=RCIM_Cubic, leaveTangent=-15.6706, time=0.9678, value=1.2074}
         <at 1 0>
        ]}}})
    Scale RGB=<0.5 0.5 0.5>
    ScaleA=true
    ScaleRGB=true
   ) {enabled=false}
   &FN_Light_Attributes(
    Color=<rgba 1 0.0757 0.8793 1>
    Create Color Attribute=true
    Create Exponent Attribute=true
    Create Position Attribute=true
    Create Radius Attribute=true
    Exponent=5
    Radius=&Lerp_Float(
     A=200
     B=500
     Alpha=&FloatFromCurve(
      FloatCurve=DataInterface() {propertyValues {Curve {keys [<at 0 0>, {interpMode=RCIM_Cubic, tangentMode=RCTM_User, time=0.5, value=0.2}, <at 0.9 1>]}}}
      Scale Curve=&Multiply_Float(A=0.01, B=$User.Radius)
     )
    )
    Scale Color=&Multiply_Float(A=8, B=$User.Fade out)
   )
  ]
  rendererClasses [&NiagaraLightRendererProperties]
  renderers [{rendererClass=&NiagaraLightRendererProperties, rendererIndex=0}]
 }
 LightningSmall001 {
  emitterUpdateScript [
   &EmitterState(
    Inactive Response=::NewEnumerator1
    Life Cycle Mode=::NewEnumerator1
    Loop Duration=1.5
    Recalculate Duration Each Loop=true
    UseLoopDelay=true
   )
   &SpawnBurst_Instantaneous(Spawn Count=3, Use Spawn Probability=false, UseLoopCountLimit=false)
   &SpawnRate(SpawnRate=&Multiply_Float(A=$User.Lightning Spawn Rate, B=2), Use Spawn Probability=true)
  ]
  particleSpawnScript [
   &InitializeParticle(
    Color Maximum=<rgba 1.559 0.1511 2 1>
    Color Minimum=<rgba 0.3905 0.032 0.8281 1>
    Color Mode=::NewEnumerator2
    Lifetime Max=0.8
    Lifetime Min=0.3
    Lifetime Mode=::NewEnumerator1
    Mesh Scale=&RandomRangeVector(Minimum=<2.1 2.1 2.1>, Maximum=<2.3 2.3 2.3>)
    Mesh Scale Mode=::NewEnumerator4
    UsePositionOffset=false
    Write Material Random=true
   )
   &CameraOffset(Camera Offset Amount=5)
   &ShapeLocation(
    Hemisphere Distribution=<1 0.7>
    MaskSpawnGroup=false
    Non Uniform Scale=<1 1 1.5>
    Offset=&RandomRangeVector(Minimum=<0 0 80>, Maximum=<0 0 180>)
    Offset Mode=::NewEnumerator0
    Override Seed=false
    Sphere Radius=1400
    Sphere Surface Distribution=1
    UseOwnerScale=true
   )
   &InitialMeshOrientation(
    Mesh Orientation Mode=::NewEnumerator2
    Rotation=&RandomRangeVector(Minimum=<0 0 0>, Maximum=<1 0 0>)
    Use Rotation Vector=true
   ) {module=InitialMeshOrientation002}
   &AddVelocity(Module.Use Falloff Distance=false, Origin Offset=<0 0 100>, Override Seed=false, Velocity Mode=::NewEnumerator1, Velocity Speed=1)
  ]
  particleUpdateScript [
   &ParticleState()
   &ScaleMeshSize(Scale Factor=&Multiply_VectorByFloat(Vector=<1 1.5 1>, Float=20))
   &ScaleColor(
    Scale Alpha=&Multiply_Float(
     A=5
     B=&FloatFromCurve(FloatCurve=DataInterface() {propertyValues {Curve {keys [<at 0 0.8 RCIM_Cubic>,<at 1 0 RCIM_Cubic>]}}})
    )
    Scale RGB=&Multiply_VectorByFloat(
     Vector=&MakeVectorFromLinearColorRGB(Color=<rgba 0.9485 0.1 1 1>)
     Float=&FloatFromCurve(
      FloatCurve=DataInterface() {propertyValues {Curve {keys [<at 0.1654 2>,<at 0.2596 0.1609>,<at 0.3956 0>,<at 0.5 1.5>,<at 0.6 1.5>,<at 0.7029 0.016>,<at 0.8029 0.016>,<at 0.9005 0.3575>,<at 1 0>]}}}
      Scale Curve=70
     )
    )
    ScaleA=true
    ScaleRGB=true
   )
   &ScaleColor(
    Scale Alpha=&Multiply_Float(
     A=&FloatFromCurve(FloatCurve=DataInterface() {propertyValues {Curve {keys [<at 0 1 RCIM_Cubic>,<at 1 0 RCIM_Cubic>]}}}, Scale Curve=3)
     B=&FN_FrontEndFloatModifier(Gameplay=1, Front End=0.8)
    )
    Scale RGB=&Multiply_VectorByFloat(
     Vector=&MakeVectorFromLinearColorRGB(Color=<rgba 0.9485 0.1 1 1>)
     Float=&FloatFromCurve(
      FloatCurve=DataInterface() {propertyValues {Curve {keys [<at 0 1>,<at 0.2 0>,<at 0.3 1>,<at 0.5073 0.0115>,<at 0.8029 0.016>,<at 0.9005 0.3575>,<at 1 0>]}}}
      Scale Curve=4
     )
    )
    ScaleA=true
    ScaleRGB=true
   ) {enabled=false, module=ScaleColor001}
   &DynamicMaterialParameters(
    Index 0 Param 1=&FloatFromCurve(FloatCurve=DataInterface() {propertyValues {Curve {keys [{tangentMode=RCTM_User, time=0, value=0.2}, {tangentMode=RCTM_User, time=1, value=1}]}}})
    Index 0 Param 2=0.7
    Index 0 Param 3=&FloatFromCurve(FloatCurve=DataInterface() {propertyValues {Curve {keys [
         <at 0 0 RCIM_Cubic>
         {arriveTangent=6.0024, interpMode=RCIM_Cubic, leaveTangent=6.0024, tangentMode=RCTM_User, time=0.4463, value=1.4389}
         {arriveTangent=-0.3684, interpMode=RCIM_Cubic, leaveTangent=-0.3684, tangentMode=RCTM_User, time=1, value=3}
        ]}}})
    Index 1 Param 1=20
    Index 1 Param 2=&RandomRangeFloat(Minimum=0.5, Maximum=1)
    Index 1 Param 3=&RandomRangeFloat(Minimum=0.6, Maximum=1.2)
    Index 1 Param 4=1
    Module.Param0WriteEnabled=true
    Module.Param0WriteEnabled1=true
    Module.Param1WriteEnabled=true
    Module.Param1WriteEnabled1=true
    Module.Param2WriteEnabled=true
    Module.Param2WriteEnabled1=true
    Module.Param3WriteEnabled=true
    Module.Param3WriteEnabled1=true
    Write Parameter Index 1=true
   )
   &SolveForcesAndVelocity(Clamp Velocity=false, Limit Acceleration=false)
   /SetVariables(Particles.Lightning glow color:LinearColor=&MakeLinearColorFromVectorAndFloat(
     "Vector (RGB)"=&MakeVectorFromLinearColorRGB(Color=<rgba 0.5967 0.075 1.5 1>)
     "Float (Alpha)"=&Multiply_Float(
      A=&FloatFromCurve(
       FloatCurve=DataInterface() {propertyValues {Curve {keys [<at 0 1 RCIM_Cubic>, {interpMode=RCIM_Cubic, leaveTangent=-3.6307, time=0.9, value=0}]}}}
       Scale Curve=3
      )
      B=1
     )
    )) {isSetParameters=true}
   /SetVariables(Particles.Glow Dy02:Vector4f=&MakeVector4(
     X=&FloatFromCurve(FloatCurve=DataInterface() {propertyValues {Curve {keys [{tangentMode=RCTM_User, time=0, value=0.2}, {tangentMode=RCTM_User, time=0.5, value=1}]}}})
     Y=0.7
     Z=0
     W=1
    )) {enabled=false, isSetParameters=true}
   /SetVariables(Particles.Glow Dy03:Vector4f=&MakeVector4(
     X=&FloatFromCurve(FloatCurve=DataInterface() {propertyValues {Curve {keys [<at 0.5 1>,<at 1 0.3>]}}})
     Y=0
     Z=0.5
     W=1
    )) {enabled=false, isSetParameters=true}
   &FN_Sprite_Attributes(
    Camera Offset=20
    Color=&MakeLinearColorFromVectorAndFloat(
     "Vector (RGB)"=&Multiply_Vector(A=&MakeVectorFromLinearColorRGB(Color=<rgba 0.3787 0.02 1 1>), B=<1 1 1>)
     "Float (Alpha)"=&Multiply_Float(
      A=&FloatFromCurve(FloatCurve=DataInterface() {propertyValues {Curve {keys [
           {interpMode=RCIM_Cubic, tangentMode=RCTM_User, time=0, value=0}
           {interpMode=RCIM_Cubic, tangentMode=RCTM_User, time=0.2, value=1}
           {tangentMode=RCTM_User, time=0.6, value=0}
          ]}}})
      B=1
     )
    )
    Create Camera Offset Attribute=true
    Create Color Attribute=true
    Create SpriteSize Attribute=true
    SpriteSize=&Multiply_Vector2D_ByFloat(
     Float=&FloatFromCurve(FloatCurve=DataInterface() {propertyValues {Curve {keys [
          {interpMode=RCIM_Cubic, tangentMode=RCTM_User, time=0, value=0}
          {interpMode=RCIM_Cubic, tangentMode=RCTM_User, time=0.2, value=1}
          {tangentMode=RCTM_User, time=0.6, value=0}
         ]}}})
     Vector2D=<30 30>
    )
    Write to Module Namespace=true
   ) {enabled=false}
  ]
  rendererClasses [&NiagaraMeshRendererProperties, &NiagaraSpriteRendererProperties]
  renderers [
   {rendererClass=&NiagaraMeshRendererProperties, rendererIndex=0}
   {rendererClass=&NiagaraMeshRendererProperties, rendererIndex=1}
   {rendererClass=&NiagaraSpriteRendererProperties, rendererIndex=2}
  ]
 }
 MeshRibbon {
  bEnabled=false
  emitterSpawnScript [/SetVariables(Emitter.MeshRibbonData=DataInterface() {propertyValues {
      bExposeCurve=true
      AlphaCurve {keys []}
      BlueCurve {keys [<at 0 1>,<at 1 1>]}
      GreenCurve {keys [<at 0 -0.5 RCIM_Cubic>, {arriveTangent=0.0331, interpMode=RCIM_Cubic, leaveTangent=0.0331, tangentMode=RCTM_User, time=1, value=-1}]}
      RedCurve {keys [<at 0 1 RCIM_Cubic>,<at 1 2 RCIM_Cubic>]}
     }}) {isSetParameters=true}]
  emitterUpdateScript [
   &EmitterState()
   &SpawnBurst_Instantaneous(Spawn Count=12, Use Spawn Probability=false, UseLoopCountLimit=false) {enabled=false}
   &SpawnRate(SpawnRate=10, Use Spawn Probability=false)
  ]
  particleSpawnScript [
   &InitializeParticle(
    Color=<rgba 5 5 5 1>
    Lifetime=&RandomRangeFloat(Minimum=1.7945, Maximum=3.5428)
    Mesh Scale=&Multiply_VectorByFloat(Vector=<1 1 2.5>, Float=7)
    Mesh Scale Mode=::NewEnumerator4
    Position Offset=<0 0 900>
    UsePositionOffset=true
    Write Material Random=true
   )
   &InitialMeshOrientation(
    Mesh Orientation Mode=::NewEnumerator2
    Orientation Vector=<0 0 -1>
    Rotation=&RandomRangeVector(Minimum=<0 0 0>, Maximum=<180 0 0>)
    Use Rotation Vector=true
   )
  ]
  particleUpdateScript [
   &ParticleState()
   &UpdateMeshOrientation(Rotation Rate=&RandomRangeFloat(Minimum=0, Maximum=0.1), Rotation Vector=<0.5 0 0>)
   &ScaleMeshSize(Scale Factor=&VectorFromFloat(Value=&FloatFromCurve(FloatCurve=DataInterface() {propertyValues {Curve {keys [<at 0 1 RCIM_Cubic>,<at 1 0.9 RCIM_Cubic>]}}})))
   &DynamicMaterialParameters(
    Index 0 Param 1=&FloatFromCurve(FloatCurve=DataInterface() {propertyValues {Curve {keys [
         {interpMode=RCIM_Cubic, tangentMode=RCTM_User, time=0, value=0.5}
         {arriveTangent=2.7589, interpMode=RCIM_Cubic, leaveTangent=2.7589, tangentMode=RCTM_User, time=0.9682, value=2}
        ]}}})
    Index 0 Param 2=&FloatFromCurve(FloatCurve=DataInterface() {propertyValues {Curve {keys [{interpMode=RCIM_Cubic, tangentMode=RCTM_User, time=0, value=0}, {interpMode=RCIM_Cubic, tangentMode=RCTM_User, time=1, value=1}]}}})
    Index 0 Param 3=&RandomRangeFloat(Minimum=3.0512, Maximum=5.4)
    Index 0 Param 4=&Add_Float(A=&RandomRangeFloat(Minimum=16, Maximum=40), B=0)
    Module.Index0Param0Write=true
    Module.Index0Param1Write=true
    Module.Index0Param2Write=true
    Module.Index0Param3Write=true
   )
   &ScaleColor(
    Linear Color Curve=DataInterface() {propertyValues {
      AlphaCurve {keys [<at 0 0 RCIM_Cubic>,<at 0.2932 3 RCIM_Cubic>,<at 1 0 RCIM_Cubic>]}
      BlueCurve {keys [<at 0 0 RCIM_Cubic>,<at 0.2925 3 RCIM_Cubic>,<at 1 0 RCIM_Cubic>]}
      GreenCurve {keys [<at 0 0 RCIM_Cubic>,<at 0.2925 3 RCIM_Cubic>,<at 1 0 RCIM_Cubic>]}
      RedCurve {keys [<at 0 0 RCIM_Cubic>,<at 0.2925 3 RCIM_Cubic>,<at 1 0 RCIM_Cubic>]}
     }}
    Scale Mode=::NewEnumerator2
   )
   &CameraOffset(Camera Offset Amount=50)
   &AccelerationForce() {enabled=false}
   &SolveForcesAndVelocity(Clamp Velocity=false, Limit Acceleration=false) {enabled=false}
  ]
  rendererClasses [&NiagaraMeshRendererProperties]
  renderers [{rendererClass=&NiagaraMeshRendererProperties, rendererIndex=0}]
 }
 Motes {
  emitterUpdateScript [
   &EmitterState()
   &SpawnBurst_Instantaneous(Spawn Count=1, Use Spawn Probability=false)
   &SpawnRate(SpawnRate=&Lerp_Float(A=0, B=100, Alpha=$User.Fade out), Use Spawn Probability=false)
  ]
  particleSpawnScript [
   &InitializeParticle(
    Color=&LerpLinearColors(
     StartColor=&RandomRangeLinearColor(Minimum=<rgba 0.4038 0.1612 3 1>, Maximum=<rgba 5 0.6988 3.2464 1>)
     EndColor=$User.InteriorColor
     LerpFactor=$System.OpenAmount
    )
    Lifetime Max=4
    Lifetime Min=2
    Lifetime Mode=::NewEnumerator1
    Mesh Scale Mode=::NewEnumerator4
    Position Mode=::NewEnumerator1
    Sprite Rotation Mode=::NewEnumerator2
    Sprite Size Mode=::NewEnumerator1
    Uniform Sprite Size Max=30
    Uniform Sprite Size Min=10
    UsePositionOffset=false
    Write Material Random=true
   )
   /SetVariables(Particles.RandomPosition:NiagaraFloat=&FloatFromCurve(
     FloatCurve=DataInterface() {propertyValues {Curve {keys [
         {arriveTangent=-5.5132, interpMode=RCIM_Cubic, leaveTangent=-3.1049, tangentMode=RCTM_User, time=0, value=1}
         {arriveTangent=-0.55, interpMode=RCIM_Cubic, leaveTangent=-0.55, tangentMode=RCTM_User, time=0.3576, value=0.374}
         {arriveTangent=-1.1196, interpMode=RCIM_Cubic, leaveTangent=-1.1196, tangentMode=RCTM_User, time=1, value=0}
        ]}}}
     CurveIndex=&RandomRangeFloat(Minimum=0, Maximum=1)
    )) {isSetParameters=true}
   /SetVariables(Particles.Position=&MakeVector(
     X=&FloatFromCurve(
      FloatCurve=DataInterface() {propertyValues {Curve {keys [
          {arriveTangent=4.8627, interpMode=RCIM_Cubic, leaveTangent=0.4743, tangentMode=RCTM_User, time=0, value=0.07}
          {arriveTangent=2.7374, interpMode=RCIM_Cubic, leaveTangent=2.7374, tangentMode=RCTM_User, time=0.2781, value=0.3179}
          {arriveTangent=0.7984, interpMode=RCIM_Cubic, leaveTangent=0.7984, tangentMode=RCTM_User, time=0.6554, value=1.0037}
          {arriveTangent=1.8113, interpMode=RCIM_Cubic, leaveTangent=0.022, tangentMode=RCTM_User, time=1, value=0.75}
         ]}}}
      CurveIndex=$Particles.RandomPosition
      Scale Curve=500
     )
     Y=&Add_Float(
      A=&FloatFromCurve(
       FloatCurve=DataInterface() {propertyValues {Curve {keys [
           {arriveTangent=-2.6903, interpMode=RCIM_Cubic, leaveTangent=-2.6903, tangentMode=RCTM_User, time=-0.0027, value=0}
           {interpMode=RCIM_Cubic, tangentMode=RCTM_User, time=0.3, value=-1}
           {interpMode=RCIM_Cubic, leaveTangent=1.4239, time=0.7, value=1}
           <at 1 0 RCIM_Cubic>
          ]}}}
       CurveIndex=$Particles.RandomPosition
       Scale Curve=0
      )
      B=&FloatFromCurve(
       FloatCurve=DataInterface() {propertyValues {Curve {keys [
           {arriveTangent=-0.8464, interpMode=RCIM_Cubic, time=0, value=0.25}
           {arriveTangent=1.1407, interpMode=RCIM_Cubic, leaveTangent=1.1407, tangentMode=RCTM_User, time=1, value=1}
          ]}}}
       CurveIndex=$Particles.RandomPosition
       Scale Curve=&Multiply_Float(A=$System.OpenAmount, B=&UniformAOrBFloat(A=-900, B=900))
      )
     )
     Z=&FloatFromCurve(
      FloatCurve=DataInterface() {propertyValues {Curve {keys [
          {arriveTangent=7.2867, interpMode=RCIM_Cubic, time=0, value=1}
          {arriveTangent=-4.3515, interpMode=RCIM_Cubic, leaveTangent=-4.3515, tangentMode=RCTM_User, time=0.2171, value=0.5268}
          {arriveTangent=-0.0104, interpMode=RCIM_Cubic, leaveTangent=-0.0104, tangentMode=RCTM_User, time=0.45, value=0}
          {arriveTangent=4.8334, interpMode=RCIM_Cubic, leaveTangent=4.8334, tangentMode=RCTM_User, time=0.8116, value=0.579}
          {arriveTangent=-2.1665, interpMode=RCIM_Cubic, leaveTangent=-2.1665, tangentMode=RCTM_User, time=0.9827, value=0.7}
         ]}}}
      CurveIndex=$Particles.RandomPosition
      Scale Curve=900
     )
    )) {isSetParameters=true}
   &SphereLocation(
    MaskSpawnGroup=false
    Override Seed=false
    Sphere Radius=&Multiply_Float(A=&Lerp_Float(A=100, B=200, Alpha=0.5), B=10)
    Surface Only=false
    UseNonUniformScale=false
    UseOffset=false
    UseSphereOrientation=false
   ) {module=SphereLocation001}
   /SetVariables(Particles.Sparkle:NiagaraFloat=&UniformAOrBFloat(A=0, Distribution Weight=0.2)) {isSetParameters=true}
   &AddVelocity(
    Scale Added Velocity=&VectorFromFloat(Value=0)
    Velocity=&RotateVectorByQuaternion(
     Vector To Rotate=&NormalizeVector(Vector To Normalize=&Subtract_Vector(A=Linked() {linkedVariable {name=Particles.Position, type=NiagaraPosition}}, B=<1200 0 500>))
     Quaternion=<0 -1 0 1>
    )
   )
  ]
  particleUpdateScript [
   &ParticleState()
   &DynamicMaterialParameters(
    Index 0 Param 1=1
    Index 0 Param 2=1
    Index 0 Param 3=0.5
    Module.Param0WriteEnabled=true
    Module.Param1WriteEnabled=true
    Module.Param2WriteEnabled=true
    Module.Param3WriteEnabled=true
   )
   &ScaleColor(
    Scale Alpha=&FloatFromCurve(
     FloatCurve=DataInterface() {propertyValues {Curve {keys [
         <at 0 0 RCIM_Cubic>
         {interpMode=RCIM_Cubic, tangentMode=RCTM_User, time=0.1, value=0.9914}
         {interpMode=RCIM_Cubic, tangentMode=RCTM_User, time=0.7, value=1}
         <at 1 0>
        ]}}}
     Scale Curve=&Lerp_Float(A=0.25, B=1, Alpha=$System.DistanceFade)
    )
    ScaleA=true
    ScaleRGB=true
   )
   &ScaleSpriteSize(Scale Factor=&Vector2DFromFloat(Value=&FloatFromCurve(
      FloatCurve=DataInterface() {propertyValues {Curve {keys [
          <at 0 0 RCIM_Cubic>
          {arriveTangent=2.4312, interpMode=RCIM_Cubic, leaveTangent=2.4312, time=0.1483, value=1}
          {arriveTangent=-0.0468, interpMode=RCIM_Cubic, leaveTangent=-0.0468, time=0.2916, value=0.709}
          {arriveTangent=-0.2796, interpMode=RCIM_Cubic, leaveTangent=-0.2796, time=0.4899, value=0.984}
          {arriveTangent=0.7494, interpMode=RCIM_Cubic, leaveTangent=0.7494, tangentMode=RCTM_User, time=0.6851, value=0.599}
          <at 1 0>
         ]}}}
      Scale Curve=&Lerp_Float(A=0.5, B=5, Alpha=0.5)
     )))
   &VortexForce(
    Influence Falloff Radius=1600
    Origin Pull Amount=-0.5
    Use Influence Falloff=true
    Vortex Axis=<0 -1 0>
    Vortex Force Amount=100
    Vortex Origin:Vector3f=<1500 0 400>
   )
   &VectorNoiseForce(Force Amount=&Lerp_Float(A=100, B=1200, Alpha=$System.OpenAmount), Module.Use Random Noise Time Interval=true)
   /SetVariables(Particles.SparkleSize:Vector2f=/Niagara/DynamicInputs/Add/Add_Vector2(
     A:Vector2f=$Particles.SpriteSize
     B:Vector2f=&Vector2DFromFloat(Value=&Multiply_Float(
       A=$Particles.Sparkle
       B=&FloatFromCurve(
        FloatCurve=DataInterface() {propertyValues {Curve {keys [
            <at 0.4 0.05 RCIM_Cubic>
            {arriveTangent=-4.4497, interpMode=RCIM_Cubic, leaveTangent=-4.4497, tangentMode=RCTM_User, time=0.45, value=1}
            <at 0.6 0.05 RCIM_Cubic>
           ]}}}
        Scale Curve=700
       )
      ))
    )) {isSetParameters=true}
   &Drag(Use Linear Drag=true, Use Rotational Drag=false)
   &GravityForce(Gravity=<0 0 50>)
   &SolveForcesAndVelocity(Clamp Velocity=false, Limit Acceleration=false)
  ]
  rendererClasses [&NiagaraSpriteRendererProperties]
  renderers [
   {rendererClass=&NiagaraSpriteRendererProperties}
   {rendererClass=&NiagaraSpriteRendererProperties, rendererIndex=1}
  ]
 }
 Reality_Bubble {
  emitterUpdateScript [
   &EmitterState(
    Inactive Response=::NewEnumerator1
    Life Cycle Mode=::NewEnumerator1
    Loop Behavior=::NewEnumerator1
    Loop Duration Mode=::NewEnumerator1
    UseLoopDelay=false
   )
   &SpawnBurst_Instantaneous(Spawn Count=1, Use Spawn Probability=false)
  ]
  particleSpawnScript [&InitializeParticle(
    Color=<rgba 0.0008 0.1365 2 1>
    Lifetime=0.2
    Mesh Scale=<3 3 3>
    Mesh Scale Mode=::NewEnumerator4
    Position Offset=<0 0 900>
    UsePositionOffset=true
    Write Material Random=true
   )]
  particleUpdateScript [
   &ParticleState(Kill Particles When Lifetime Has Elapsed=false, Let Infinitely Lived Particles Die When Emitter Deactivates=true)
   &ScaleMeshSize(Scale Factor=&VectorFromFloat(Value=10))
   &UpdateMeshOrientation(Rotation Rate=0.2, Rotation Vector=<1 1 1>) {enabled=false}
   &DynamicMaterialParameters(
    Index 0 Param 1=&Lerp_Float(A=1, B=3, Alpha=$System.DestabilizeAmount)
    Module.Param0WriteEnabled=true
    Module.Param1WriteEnabled=true
    Module.Param2WriteEnabled=true
    Module.Param3WriteEnabled=true
   )
   &ScaleColor(
    Scale RGB=&VectorFromFloat(Value=&Lerp_Float(
      A=1
      B=300
      Alpha=&FloatFromCurve(
       FloatCurve=DataInterface() {propertyValues {Curve {keys [
           {interpMode=RCIM_Cubic, tangentMode=RCTM_User, time=0.77, value=0}
           {interpMode=RCIM_Cubic, tangentMode=RCTM_User, time=0.8, value=1}
           {tangentMode=RCTM_User, time=1, value=0}
          ]}}}
       CurveIndex=$System.Forming
      )
     ))
    ScaleA=true
    ScaleRGB=true
   ) {enabled=false}
   &ScaleColor(
    Scale RGB=&VectorFromFloat(Value=&MakeCustomFloatFromBool(
      Bool=true
      True Float=1
      False Float=&Lerp_Float(
       A=1
       B=200
       Alpha=&FloatFromCurve(
        FloatCurve=DataInterface() {propertyValues {Curve {keys [{arriveTangent=-3.0621, interpMode=RCIM_Cubic, leaveTangent=-3.0621, tangentMode=RCTM_User, time=0, value=1}, <at 0.4 0 RCIM_Cubic>]}}}
        CurveIndex=$Emitter.Age
       )
      )
     ))
    ScaleA=true
    ScaleRGB=true
   ) {enabled=false, module=ScaleColor001}
   &ScaleColor(Scale Alpha=&Lerp_Float(A=0, B=1, Alpha=$User.Fade out), ScaleA=true, ScaleRGB=true) {module=ScaleColor002}
  ]
  rendererClasses [&NiagaraMeshRendererProperties]
  renderers [{rendererClass=&NiagaraMeshRendererProperties, rendererIndex=0}]
 }
 Shimmers {
  bEnabled=false
  emitterUpdateScript [
   &EmitterState()
   &SpawnRate(SpawnRate=&Lerp_Float(A=5, B=10, Alpha=$User.WeldProgress), Use Spawn Probability=false)
  ]
  particleSpawnScript [
   &InitializeParticle(
    Color=&LerpLinearColors(
     StartColor=$User.InteriorColor
     EndColor=<rgba 0.9714 0.5 1 0.25>
     LerpFactor=&Lerp_Float(A=0.7, B=1, Alpha=$System.DestabilizeAmount)
    )
    Lifetime Max=&Lerp_Float(A=5, B=2, Alpha=$User.WeldProgress)
    Lifetime Min:NiagaraFloat=&Lerp_Float(A=3, B=2, Alpha=$User.WeldProgress)
    Lifetime Mode=::NewEnumerator1
    Sprite Rotation Mode=::NewEnumerator2
    Sprite Size Mode=::NewEnumerator1
    Uniform Sprite Size Max=350
    Uniform Sprite Size Min=250
    UsePositionOffset=false
    Write Material Random=true
   )
   /SetVariables(Particles.Position=$Engine.Emitter.SimulationPosition) {isSetParameters=true}
   &AddVelocityFromPoint(
    Module.Use Falloff Distance=false
    Origin Offset=&MakeVector(X=&Lerp_Float(A=500, B=50, Alpha=1), Y=0, Z=1000)
    Velocity Strength=1000
   )
   &CameraOffset(Camera Offset Amount=500)
   &SphereLocation(
    Hemisphere Z=true
    MaskSpawnGroup=false
    Non Uniform Scale=<10 10 10>
    Offset=<0 0 1500>
    Override Seed=false
    Sphere Radius=&Lerp_Float(A=100, B=500, Alpha=$System.OpenAmount)
    Surface Only=false
    UseNonUniformScale=true
    UseOffset=true
    UseSphereOrientation=false
   )
  ]
  particleUpdateScript [
   &ParticleState()
   &DynamicMaterialParameters(
    Index 0 Param 1=$Particles.MaterialRandom
    Index 0 Param 2=&RandomRangeFloat(Minimum=0, Maximum=1)
    Index 0 Param 3=&Lerp_Float(A=0, B=2, Alpha=$System.DestabilizeAmount)
    Module.Param0WriteEnabled=true
    Module.Param1WriteEnabled=true
    Module.Param2WriteEnabled=true
    Module.Param3WriteEnabled=false
   )
   &PointForce(
    Force Falloff Distance=1500
    Force Origin=<-100 0 0>
    Force Strength=&Lerp_Float(A=0, B=100, Alpha=&Multiply_Float(A=$System.OpenAmount, B=$Particles.NormalizedAge))
    Module.Use Falloff=true
    Module.UseLinearForce=false
   )
   &Drag(Drag=&Lerp_Float(A=0.5, B=3, Alpha=1), Use Linear Drag=true, Use Rotational Drag=false)
   &ScaleColor(
    Scale Alpha=&FloatFromCurve(
     FloatCurve=DataInterface() {propertyValues {Curve {keys [<at 0 0 RCIM_Cubic>,<at 0.35 1 RCIM_Cubic>,<at 1 0>]}}}
     Scale Curve=&Lerp_Float(A=0.22, B=1, Alpha=$System.OpenAmount)
    )
    Scale RGB=<2 0.5 5>
    ScaleA=true
    ScaleRGB=true
   )
   &ScaleSpriteSize(Scale Factor=&Vector2DFromFloat(Value=&FloatFromCurve(
      FloatCurve=DataInterface() {propertyValues {Curve {keys [<at 0 0.7 RCIM_Cubic>, {interpMode=RCIM_Cubic, tangentMode=RCTM_User, time=1, value=1}]}}}
      Scale Curve=&Lerp_Float(A=5, B=1, Alpha=$System.OpenAmount)
     )))
   &SolveForcesAndVelocity(Clamp Velocity=false, Limit Acceleration=false)
  ]
  rendererClasses [&NiagaraSpriteRendererProperties]
  renderers [{rendererClass=&NiagaraSpriteRendererProperties}]
 }
 Shimmers001 {
  simTarget=GPUComputeSim
  emitterUpdateScript [
   &EmitterState(
    Enable Distance Culling=true
    MaxDistance=10000
    Min Distance Response=::SleepAndLetParticlesFinish
    Module.bUseMaxDistance=true
    Module.bUseMinDistance=true
    Scalability Mode=::NewEnumerator1
   )
   &SpawnRate(SpawnRate=&Lerp_Float(A=0, B=200, Alpha=$User.Fade out), Use Spawn Probability=true)
  ]
  particleSpawnScript [
   /SetVariables(
    Particles.ColorEnd:LinearColor=&RandomRangeLinearColor(Minimum=<rgba 0.25 0 0.138 1>, Maximum=<rgba 0.0466 0 0.25 1>)
    Particles.ColorStart:LinearColor=&RandomRangeLinearColor(Minimum=<rgba 0.36 0.3 0.45 1>, Maximum=<rgba 0.6 0.68 0.74 1>)
   ) {isSetParameters=true}
   &InitializeParticle(
    Lifetime Mode=::NewEnumerator1
    Position Mode=::NewEnumerator1
    Sprite Rotation Mode=::NewEnumerator2
    Sprite Size=&RandomRangeVector2D(Minimum=<400 400>, Maximum=<800 800>, Evaluation Type=::NewEnumerator1)
    Sprite Size Mode=::NewEnumerator4
    UsePositionOffset=false
    Write Material Random=true
   )
   &AddVelocity(Coordinate Space=::World, Velocity=&RandomRangeVector(Minimum=<-100 -100 -100>, Maximum=<100 100 500>))
   &DynamicMaterialParameters(
    Index 0 Param 1=&RandomRangeFloat(Minimum=0, Maximum=1, Evaluation Type=::NewEnumerator1)
    Index 0 Param 2=&RandomRangeFloat(Minimum=0.4, Maximum=0.6, Evaluation Type=::NewEnumerator1)
    Index 0 Param 3=0.2
    Module.Param0WriteEnabled=true
    Module.Param1WriteEnabled=true
    Module.Param2WriteEnabled=true
    Module.Param3WriteEnabled=true
   )
   &SphereLocation(
    MaskSpawnGroup=false
    Non Uniform Scale=<2 2 3>
    Offset=<0 0 200>
    Override Seed=false
    Sphere Radius=800
    Surface Only=false
    UseNonUniformScale=true
    UseOffset=true
    UseSphereOrientation=false
   )
  ]
  particleUpdateScript [
   &ParticleState()
   &SolveForcesAndVelocity(Clamp Velocity=false, Limit Acceleration=false)
   &SpriteRotationRate(Rotation Rate=&RandomRangeFloat(Minimum=-20, Maximum=20))
   &ScaleColor(
    Scale Alpha=&FloatFromCurve(
     FloatCurve=DataInterface() {propertyValues {Curve {keys [{arriveTangent=1.0833, interpMode=RCIM_Cubic, leaveTangent=1.0833, time=0, value=0}, <at 0.6 0.65 RCIM_Cubic>, <at 1 0>]}}}
     Scale Curve=&RandomRangeFloat(Minimum=0.5, Maximum=1)
    )
    Scale RGB=&MakeVectorFromLinearColorRGB(Color=<rgba 0.589 0.1398 1 1>)
    ScaleA=true
    ScaleRGB=true
   ) {module=ScaleColor001}
   &ScaleSpriteSize(
    Uniform Curve Scale=2
    Uniform Curve Sprite Scale=DataInterface() {propertyValues {Curve {keys [<at 0 0.7 RCIM_Cubic>, {interpMode=RCIM_Cubic, tangentMode=RCTM_Break, time=1, value=1}]}}}
   )
  ]
  rendererClasses [&NiagaraSpriteRendererProperties]
  renderers [{rendererClass=&NiagaraSpriteRendererProperties}]
 }
}
summary {
 systemName=NS_LimeRock_Treehouse_FX
 emitters [
  {bEnabled=false, emitterName=Shimmers, rendererClasses [&NiagaraSpriteRendererProperties]}
  {emitterName=Motes, rendererClasses [&NiagaraSpriteRendererProperties]}
  {emitterName=Shimmers001, simTarget=GPUComputeSim, rendererClasses [&NiagaraSpriteRendererProperties]}
  {
   emitterName=FlareLightShockwave
   rendererClasses [
    &NiagaraSpriteRendererProperties
    &NiagaraLightRendererProperties
    &NiagaraMeshRendererProperties
   ]
  }
  {
   emitterName=LightningSmall001
   rendererClasses [&NiagaraMeshRendererProperties, &NiagaraSpriteRendererProperties]
  }
  {emitterName=Reality_Bubble, rendererClasses [&NiagaraMeshRendererProperties]}
  {emitterName=Light001, rendererClasses [&NiagaraLightRendererProperties]}
  {bEnabled=false, emitterName=MeshRibbon, rendererClasses [&NiagaraMeshRendererProperties]}
 ]
 userVariables [
  User.A_SpawnRate {type=NiagaraFloat, defaultValue=0, description=""}
  User.Alpha_Scale Alpha {type=NiagaraFloat, defaultValue=0, description=""}
  User.Destabilize Amount {type=NiagaraFloat, defaultValue=0, description=""}
  User.Fade out {type=NiagaraFloat, defaultValue=1, description=""}
  User.InteriorColor {type=LinearColor, description="", defaultValue {struct=LinearColor, value=<rgba 1 1 1 1>}}
  User.IsSummon {type=NiagaraBool, defaultValue=true}
  User.Lightning Spawn Rate {type=NiagaraFloat, defaultValue=2, description=""}
  User.OpenAmount {type=NiagaraFloat, defaultValue=0, description=""}
  User.Radius {type=NiagaraFloat, defaultValue=1000, description=""}
  User.Spline {
   type=Object
   description=""
   defaultValue {struct=NiagaraExt_VariableValue_Object, value {object=null, objectClass=&Object}}
  }
  User.Spline_ZipPath {
   type=NiagaraDataInterfaceSpline
   description=""
   defaultValue {
    struct=NiagaraExt_VariableValue_DataInterface
    value {
     dataInterface=&self:NiagaraDataInterfaceSpline_0
     dataInterfaceClass=/Script/Niagara.NiagaraDataInterfaceSpline
    }
   }
  }
 ]
}
