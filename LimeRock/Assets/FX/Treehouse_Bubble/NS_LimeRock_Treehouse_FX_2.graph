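# Niagara system for the LimeRock treehouse energy bubble effect, layering pulsing waves, lightning, motes, shimmers, decals, spline meshes and a light around the tree.
# Lightning fills the canopy and trunk, while shimmers spawn in a top sphere and on a placeholder treehouse mesh.
# Several emitters blend magenta to blue by particle height against System.Height.
# User parameters drive spawn rates, brightness, fading and color switches, with curves animating dynamic material parameters.
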NiagaraSystem /LimeRock/Assets/FX/Treehouse_Bubble/NS_LimeRock_Treehouse_FX_2
dependencies {
 usedDataInterfaces [&NiagaraDataInterfaceCurve, &NiagaraDataInterfaceStaticMesh]
 usedDynamicInputs []
 usedModules [
  &SystemState
  &self:SystemSpawnScript.SystemScriptSource.SystemScriptGraph.NiagaraNodeAssignment_2.SetVariables_#1
 ]
}
emitters {
 BigWave {
  emitterUpdateScript [
   &EmitterState()
   &SpawnRate(SpawnRate=&Multiply_Float(A=3, B=$User.Energy_SpawnRate), Use Spawn Probability=false)
  ]
  particleSpawnScript [
   &InitializeParticle(
    Lifetime=&Multiply_Float(A=&RandomRangeFloat(Minimum=0.7, Maximum=1.2), B=$User.Energy_SpawnTimeMult)
    Mesh Scale Mode=::NewEnumerator1
    Mesh Uniform Scale Max=40
    Mesh Uniform Scale Min=20
    Position=<0 0 1000>
    Position Mode=::NewEnumerator1
    UsePositionOffset=false
    Write Material Random=true
   )
   /SetVariables(Particles.VisibilityTag=&SpawnGroup()) {isSetParameters=true}
   &InitialMeshOrientation(
    Mesh Orientation Mode=::NewEnumerator1
    Rotation=&RandomRangeVector(Minimum=<-0.1 -0.1 0>, Maximum=<0.1 0.1 1>)
    Use Rotation Vector=true
   )
  ]
  particleUpdateScript [
   &ParticleState()
   &DynamicMaterialParameters(
    Index 0 Param 1=&FloatFromCurve(FloatCurve=DataInterface() {propertyValues {Curve {keys [{interpMode=RCIM_Cubic, tangentMode=RCTM_User, time=0.5, value=0}, {interpMode=RCIM_Cubic, tangentMode=RCTM_User, time=1, value=1}]}}})
    Index 0 Param 3=1
    Module.Param0WriteEnabled=true
    Module.Param1WriteEnabled=true
    Module.Param2WriteEnabled=true
    Module.Param3WriteEnabled=true
   )
   &ScaleMeshSize(Scale Factor=&Multiply_VectorByFloat(
     Vector=<1 1 1.5>
     Float=&FloatFromCurve(
      FloatCurve=DataInterface() {propertyValues {Curve {
         defaultValue=0
         keys [{interpMode=RCIM_Cubic, tangentMode=RCTM_Break, time=0, value=1}, {interpMode=RCIM_Cubic, tangentMode=RCTM_Break, time=1, value=0}]
        }}}
      Scale Curve=1.6
     )
    ))
   &UpdateMeshOrientation(
    Rotation Rate=&FloatFromCurve(FloatCurve=DataInterface() {propertyValues {Curve {keys [{arriveTangent=-2.0983, interpMode=RCIM_Cubic, leaveTangent=-2.0983, tangentMode=RCTM_User, time=0, value=1}, <at 1 0>]}}})
    Rotation Vector=<0 0 1>
   )
   &ScaleColor(
    Scale Alpha=&FloatFromCurve(
     FloatCurve=DataInterface() {propertyValues {Curve {keys [
         {arriveTangent=8.2356, interpMode=RCIM_Cubic, leaveTangent=8.2356, tangentMode=RCTM_User, time=0, value=0}
         {arriveTangent=0.1969, interpMode=RCIM_Cubic, leaveTangent=0.1969, tangentMode=RCTM_User, time=0.22, value=1}
         {arriveTangent=-1.2821, interpMode=RCIM_Cubic, leaveTangent=-1.2821, time=0.5, value=0.6}
         {arriveTangent=-0.7636, interpMode=RCIM_Cubic, leaveTangent=-0.7636, tangentMode=RCTM_User, time=1, value=0}
        ]}}}
     Scale Curve=0.3
    )
    Scale RGB=&MakeVectorFromLinearColorRGB(Color=&Multiply_LinearColorByFloat(Linear Color=<rgba 0.829 0.042 1 1>, Float=1))
    ScaleA=true
    ScaleRGB=true
   )
  ]
  rendererClasses [&NiagaraMeshRendererProperties]
  renderers [{rendererClass=&NiagaraMeshRendererProperties, rendererIndex=0}]
 }
 Decal_Energy_Ground_Right001 {
  emitterUpdateScript [
   &EmitterState(Life Cycle Mode=::NewEnumerator1, UseLoopDelay=false)
   &SpawnBurst_Instantaneous(Use Spawn Probability=false, UseLoopCountLimit=false)
  ]
  particleSpawnScript [
   &InitializeParticle(Lifetime=5, Position=<-200 0 100>, Position Mode=::NewEnumerator1, UsePositionOffset=false, Write Material Random=true)
   /SetVariables(Particles.DecalOrientation:Quat4f=&OrientMeshToVector(Look At Direction=<1 0 0>, Reference Axis Vector=<1 0 0>)) {isSetParameters=true}
   /SetVariables(Particles.DecalSize:Vector3f=<2100 2100 400>) {isSetParameters=true}
  ]
  particleUpdateScript [
   &ParticleState()
   &ScaleColor(
    Scale Alpha=$User.GroundDecal_Intensity
    Scale RGB=&VectorFromFloat(Value=$User.LightColor_Switch)
    ScaleA=true
    ScaleRGB=true
   )
  ]
  rendererClasses [&NiagaraDecalRendererProperties]
  renderers [
   {rendererClass=&NiagaraDecalRendererProperties, rendererIndex=0}
   {rendererClass=&NiagaraDecalRendererProperties, rendererIndex=1}
  ]
 }
 Light001 {
  emitterUpdateScript [
   &EmitterState(
    Inactive Response=::NewEnumerator1
    Life Cycle Mode=::NewEnumerator1
    Loop Behavior=::NewEnumerator1
    Loop Delay=2
    Loop Duration Mode=::NewEnumerator1
    UseLoopDelay=false
   )
   &SpawnBurst_Instantaneous(Spawn Count=1, Use Spawn Probability=false)
  ]
  particleSpawnScript [&InitializeParticle(Color Mode=::NewEnumerator0, Position Offset=<0 0 300>, UsePositionOffset=true, Write Material Random=false) {module=InitializeParticle001}]
  particleUpdateScript [
   &ParticleState(Kill Particles When Lifetime Has Elapsed=false)
   &FN_Light_Attributes(
    Color=&LerpLinearColors(StartColor=<rgba 0.936 0.101 1 1>, EndColor=<rgba 0.2 0.3 1 1>, LerpFactor=$User.LightColor_Switch)
    Create Color Attribute=true
    Create Exponent Attribute=true
    Create Radius Attribute=true
    Exponent=5
    Radius=&Lerp_Float(
     A=200
     B=500
     Alpha=&FloatFromCurve(
      FloatCurve=DataInterface() {propertyValues {Curve {keys [<at 0 0>, {interpMode=RCIM_Cubic, tangentMode=RCTM_User, time=0.5, value=0.2}, <at 0.9 1>]}}}
      Scale Curve=&Multiply_Float(A=3, B=$User.Light_Radius)
     )
    )
    Scale Color=&Multiply_Float(A=$User.Light_Intensity, B=5)
   )
  ]
  rendererClasses [&NiagaraLightRendererProperties]
  renderers [{rendererClass=&NiagaraLightRendererProperties, rendererIndex=0}]
 }
 LightningSmall_Top {
  emitterUpdateScript [
   &EmitterState(
    Inactive Response=::NewEnumerator1
    Life Cycle Mode=::NewEnumerator1
    Loop Duration=1.5
    Recalculate Duration Each Loop=true
    UseLoopDelay=false
   )
   &SpawnRate(SpawnRate=&Multiply_Float(A=$User.Lightning Spawn Rate, B=1.2), Use Spawn Probability=true)
  ]
  particleSpawnScript [
   &InitializeParticle(
    AdjustAlpha=false
    AdjustHue=true
    AdjustSaturation=true
    AdjustValue=true
    Color=$User.OverallColorSwitch
    Color Mode=::NewEnumerator3
    Lifetime Max=0.8
    Lifetime Min=0.3
    Lifetime Mode=::NewEnumerator1
    Mesh Scale=&RandomRangeVector(Minimum=<2.1 2.1 2.1>, Maximum=<2.3 2.3 2.3>)
    Mesh Scale Mode=::NewEnumerator4
    UsePositionOffset=false
    Write Material Random=true
   )
   &CameraOffset(Camera Offset Amount=5)
   &ShapeLocation(
    Hemisphere Distribution=<1 0.7>
    MaskSpawnGroup=false
    Offset=<0 0 1400>
    Offset Mode=::NewEnumerator0
    Override Seed=false
    Sphere Radius=800
    UseOwnerScale=true
   )
   &InitialMeshOrientation(
    Mesh Orientation Mode=::NewEnumerator2
    Rotation=&RandomRangeVector(Minimum=<0 0 0>, Maximum=<1 0 0>)
    Use Rotation Vector=true
   ) {module=InitialMeshOrientation002}
   &AddVelocity(Module.Use Falloff Distance=false, Origin Offset=<0 0 100>, Override Seed=false, Velocity Mode=::NewEnumerator1, Velocity Speed=1)
  ]
  particleUpdateScript [
   &ParticleState()
   /SetVariables(Particles.HeightLerp:NiagaraFloat=&FloatFromCurve(
     FloatCurve=DataInterface() {propertyValues {Curve {keys [{interpMode=RCIM_Cubic, tangentMode=RCTM_Break, time=0, value=1}, {interpMode=RCIM_Cubic, tangentMode=RCTM_Break, time=200, value=0}]}}}
     CurveIndex=&Subtract_Float(
      A=$System.Height
      B=&MakeFloatFromVector(Vector=Linked() {linkedVariable {name=Particles.Position, type=NiagaraPosition}}, Channel=::NewEnumerator2)
     )
    )) {isSetParameters=true, module=SetVariables_#2}
   &ScaleMeshSize(Scale Factor=&Multiply_VectorByFloat(Vector=<1 1.5 1>, Float=10))
   &ScaleColor(
    Scale Alpha=&Multiply_Float(
     A=&Multiply_Float(A=1, B=$User.TreeEnergy_Brightness)
     B=&FloatFromCurve(FloatCurve=DataInterface() {propertyValues {Curve {keys [<at 0 0.8 RCIM_Cubic>,<at 1 0 RCIM_Cubic>]}}})
    )
    Scale RGB=&Multiply_VectorByFloat(
     Vector=&Multiply_VectorByFloat(
      Vector=&MakeVectorFromLinearColorRGB(Color=&LerpLinearColors(StartColor=<rgba 0.936 0.101 1 1>, EndColor=<rgba 0.2 0.3 1 1>, LerpFactor=$Particles.HeightLerp))
      Float=1
     )
     Float=&FloatFromCurve(
      FloatCurve=DataInterface() {propertyValues {Curve {keys [<at 0 1>,<at 0.2 0>,<at 0.3 1>,<at 0.5073 0.0115>,<at 0.8029 0.016>,<at 0.9005 0.3575>,<at 1 0>]}}}
      Scale Curve=&Multiply_Float(A=4, B=$User.TreeEnergy_Brightness)
     )
    )
    ScaleA=true
    ScaleRGB=true
   )
   &DynamicMaterialParameters(
    Index 0 Param 1=&FloatFromCurve(FloatCurve=DataInterface() {propertyValues {Curve {keys [{tangentMode=RCTM_User, time=0, value=0.2}, {tangentMode=RCTM_User, time=1, value=1}]}}})
    Index 0 Param 2=0.7
    Index 0 Param 3=&FloatFromCurve(FloatCurve=DataInterface() {propertyValues {Curve {keys [
         <at 0 0 RCIM_Cubic>
         {arriveTangent=6.0024, interpMode=RCIM_Cubic, leaveTangent=6.0024, tangentMode=RCTM_User, time=0.4463, value=1.4389}
         {arriveTangent=-0.3684, interpMode=RCIM_Cubic, leaveTangent=-0.3684, tangentMode=RCTM_User, time=1, value=3}
        ]}}})
    Index 1 Param 1=20
    Index 1 Param 2=&RandomRangeFloat(Minimum=0.5, Maximum=1)
    Index 1 Param 3=&RandomRangeFloat(Minimum=0.6, Maximum=1.2)
    Index 1 Param 4=1
    Module.Param0WriteEnabled=true
    Module.Param0WriteEnabled1=true
    Module.Param1WriteEnabled=true
    Module.Param1WriteEnabled1=true
    Module.Param2WriteEnabled=true
    Module.Param2WriteEnabled1=true
    Module.Param3WriteEnabled=true
    Module.Param3WriteEnabled1=true
    Write Parameter Index 1=true
   )
   &SolveForcesAndVelocity(Clamp Velocity=false, Limit Acceleration=false)
   /SetVariables(Particles.Lightning glow color:LinearColor=&MakeLinearColorFromVectorAndFloat(
     "Vector (RGB)"=&Multiply_VectorByFloat(
      Vector=&MakeVectorFromLinearColorRGB(Color=&LerpLinearColors(StartColor=<rgba 0.936 0.101 1 1>, EndColor=<rgba 0.2 0.3 1 1>, LerpFactor=$Particles.HeightLerp))
      Float=1
     )
     "Float (Alpha)"=&Multiply_Float(
      A=&FloatFromCurve(
       FloatCurve=DataInterface() {propertyValues {Curve {keys [<at 0 1 RCIM_Cubic>, {interpMode=RCIM_Cubic, leaveTangent=-3.6307, time=0.9, value=0}]}}}
       Scale Curve=3
      )
      B=$User.TreeEnergy_Brightness
     )
    )) {isSetParameters=true, module=SetVariables_#3}
  ]
  rendererClasses [&NiagaraMeshRendererProperties]
  renderers [
   {rendererClass=&NiagaraMeshRendererProperties, rendererIndex=0}
   {rendererClass=&NiagaraMeshRendererProperties, rendererIndex=1}
  ]
 }
 LightningSmall_Trunk {
  emitterUpdateScript [
   &EmitterState(
    Inactive Response=::NewEnumerator1
    Life Cycle Mode=::NewEnumerator1
    Loop Duration=1.5
    Recalculate Duration Each Loop=true
    UseLoopDelay=false
   )
   &SpawnRate(
    SpawnRate=&Multiply_Float(A=$User.Lightning Spawn Rate, B=&Multiply_Float(A=$User.TreeTrunk_LightningRate, B=1))
    Use Spawn Probability=true
   )
  ]
  particleSpawnScript [
   &InitializeParticle(
    AdjustAlpha=false
    AdjustHue=true
    AdjustSaturation=true
    AdjustValue=true
    Color=$User.OverallColorSwitch
    Color Mode=::NewEnumerator3
    Lifetime Max=0.8
    Lifetime Min=0.3
    Lifetime Mode=::NewEnumerator1
    Mesh Scale=&RandomRangeVector(Minimum=<2.1 2.1 2.1>, Maximum=<2.3 2.3 2.3>)
    Mesh Scale Mode=::NewEnumerator4
    UsePositionOffset=false
    Write Material Random=true
   )
   &CameraOffset(Camera Offset Amount=5)
   &ShapeLocation(
    Cylinder Height=900
    Cylinder Radius=200
    MaskSpawnGroup=false
    Non Uniform Scale=<1 1 1.2>
    Offset=<0 0 600>
    Offset Mode=::NewEnumerator0
    Override Seed=false
    Shape Primitive=::NewEnumerator1
    Surface Only=false
    UseOwnerScale=true
   )
   &InitialMeshOrientation(
    Mesh Orientation Mode=::NewEnumerator2
    Rotation=&RandomRangeVector(Minimum=<0 0 0>, Maximum=<1 0 0>)
    Use Rotation Vector=true
   ) {module=InitialMeshOrientation002}
   &AddVelocity(Module.Use Falloff Distance=false, Origin Offset=<0 0 100>, Override Seed=false, Velocity Mode=::NewEnumerator1, Velocity Speed=1)
  ]
  particleUpdateScript [
   &ParticleState()
   /SetVariables(Particles.HeightLerp:NiagaraFloat=&FloatFromCurve(
     FloatCurve=DataInterface() {propertyValues {Curve {keys [{interpMode=RCIM_Cubic, tangentMode=RCTM_Break, time=0, value=1}, {interpMode=RCIM_Cubic, tangentMode=RCTM_Break, time=200, value=0}]}}}
     CurveIndex=&Subtract_Float(
      A=$System.Height
      B=&MakeFloatFromVector(Vector=Linked() {linkedVariable {name=Particles.Position, type=NiagaraPosition}}, Channel=::NewEnumerator2)
     )
    )) {isSetParameters=true, module=SetVariables_#2}
   &ScaleMeshSize(Scale Factor=&Multiply_VectorByFloat(Vector=<1 1.5 1>, Float=10))
   &ScaleColor(
    Scale Alpha=&Multiply_Float(
     A=&Multiply_Float(A=1, B=$User.TreeEnergy_Brightness)
     B=&FloatFromCurve(FloatCurve=DataInterface() {propertyValues {Curve {keys [<at 0 0.8 RCIM_Cubic>,<at 1 0 RCIM_Cubic>]}}})
    )
    Scale RGB=&Multiply_VectorByFloat(
     Vector=&MakeVectorFromLinearColorRGB(Color=&LerpLinearColors(StartColor=<rgba 0.936 0.101 1 1>, EndColor=<rgba 0.2 0.3 1 1>, LerpFactor=$Particles.HeightLerp))
     Float=&FloatFromCurve(
      FloatCurve=DataInterface() {propertyValues {Curve {keys [<at 0.1654 2>,<at 0.2596 0.1609>,<at 0.3956 0>,<at 0.5 1.5>,<at 0.6 1.5>,<at 0.7029 0.016>,<at 0.8029 0.016>,<at 0.9005 0.3575>,<at 1 0>]}}}
      Scale Curve=&Multiply_Float(A=70, B=$User.TreeEnergy_Brightness)
     )
    )
    ScaleA=true
    ScaleRGB=true
   )
   &DynamicMaterialParameters(
    Index 0 Param 1=&FloatFromCurve(FloatCurve=DataInterface() {propertyValues {Curve {keys [{tangentMode=RCTM_User, time=0, value=0.2}, {tangentMode=RCTM_User, time=1, value=1}]}}})
    Index 0 Param 2=0.7
    Index 0 Param 3=&FloatFromCurve(FloatCurve=DataInterface() {propertyValues {Curve {keys [
         <at 0 0 RCIM_Cubic>
         {arriveTangent=6.0024, interpMode=RCIM_Cubic, leaveTangent=6.0024, tangentMode=RCTM_User, time=0.4463, value=1.4389}
         {arriveTangent=-0.3684, interpMode=RCIM_Cubic, leaveTangent=-0.3684, tangentMode=RCTM_User, time=1, value=3}
        ]}}})
    Index 1 Param 1=20
    Index 1 Param 2=&RandomRangeFloat(Minimum=0.5, Maximum=1)
    Index 1 Param 3=&RandomRangeFloat(Minimum=0.6, Maximum=1.2)
    Index 1 Param 4=1
    Module.Param0WriteEnabled=true
    Module.Param0WriteEnabled1=true
    Module.Param1WriteEnabled=true
    Module.Param1WriteEnabled1=true
    Module.Param2WriteEnabled=true
    Module.Param2WriteEnabled1=true
    Module.Param3WriteEnabled=true
    Module.Param3WriteEnabled1=true
    Write Parameter Index 1=true
   )
   &SolveForcesAndVelocity(Clamp Velocity=false, Limit Acceleration=false)
   /SetVariables(Particles.Lightning glow color:LinearColor=&MakeLinearColorFromVectorAndFloat(
     "Vector (RGB)"=&MakeVectorFromLinearColorRGB(Color=&LerpLinearColors(StartColor=<rgba 0.936 0.101 1 1>, EndColor=<rgba 0.2 0.3 1 1>, LerpFactor=$Particles.HeightLerp))
     "Float (Alpha)"=&Multiply_Float(
      A=&FloatFromCurve(FloatCurve=DataInterface() {propertyValues {Curve {keys [<at 0 1 RCIM_Cubic>, {interpMode=RCIM_Cubic, leaveTangent=-3.6307, time=0.9, value=0}]}}})
      B=$User.TreeEnergy_Brightness
     )
    )) {isSetParameters=true, module=SetVariables_#3}
  ]
  rendererClasses [&NiagaraMeshRendererProperties]
  renderers [
   {rendererClass=&NiagaraMeshRendererProperties, rendererIndex=0}
   {rendererClass=&NiagaraMeshRendererProperties, rendererIndex=1}
  ]
 }
 Motes {
  emitterUpdateScript [
   &EmitterState()
   &SpawnBurst_Instantaneous(Spawn Count=1, Use Spawn Probability=false)
   &SpawnRate(SpawnRate=&Lerp_Float(A=0, B=100, Alpha=$User.Fade out), Use Spawn Probability=false)
  ]
  particleSpawnScript [
   &InitializeParticle(
    AdjustAlpha=false
    AdjustHue=true
    AdjustSaturation=true
    AdjustValue=true
    Color Mode=::NewEnumerator3
    Hue Shift Range:Vector2f=<0.1 0.2>
    Lifetime Max=4
    Lifetime Min=2
    Lifetime Mode=::NewEnumerator1
    Position Mode=::NewEnumerator1
    Saturation Range:Vector2f=<0.8 2>
    Sprite Rotation Mode=::NewEnumerator2
    Sprite Size Mode=::NewEnumerator1
    Uniform Sprite Size Max=30
    Uniform Sprite Size Min=10
    UsePositionOffset=false
    Value Range:Vector2f=<0.8 2>
    Write Material Random=true
   )
   /SetVariables(Particles.RandomPosition:NiagaraFloat=&FloatFromCurve(
     FloatCurve=DataInterface() {propertyValues {Curve {keys [
         {arriveTangent=-5.5132, interpMode=RCIM_Cubic, leaveTangent=-3.1049, tangentMode=RCTM_User, time=0, value=1}
         {arriveTangent=-0.55, interpMode=RCIM_Cubic, leaveTangent=-0.55, tangentMode=RCTM_User, time=0.3576, value=0.374}
         {arriveTangent=-1.1196, interpMode=RCIM_Cubic, leaveTangent=-1.1196, tangentMode=RCTM_User, time=1, value=0}
        ]}}}
     CurveIndex=&RandomRangeFloat(Minimum=0, Maximum=1)
    )) {isSetParameters=true}
   /SetVariables(Particles.Position=&MakeVector(
     X=&FloatFromCurve(
      FloatCurve=DataInterface() {propertyValues {Curve {keys [
          {arriveTangent=4.8627, interpMode=RCIM_Cubic, leaveTangent=0.4743, tangentMode=RCTM_User, time=0, value=0.07}
          {arriveTangent=2.7374, interpMode=RCIM_Cubic, leaveTangent=2.7374, tangentMode=RCTM_User, time=0.2781, value=0.3179}
          {arriveTangent=0.7984, interpMode=RCIM_Cubic, leaveTangent=0.7984, tangentMode=RCTM_User, time=0.6554, value=1.0037}
          {arriveTangent=1.8113, interpMode=RCIM_Cubic, leaveTangent=0.022, tangentMode=RCTM_User, time=1, value=0.75}
         ]}}}
      CurveIndex=$Particles.RandomPosition
      Scale Curve=500
     )
     Y=&Add_Float(
      A=&FloatFromCurve(
       FloatCurve=DataInterface() {propertyValues {Curve {keys [
           {arriveTangent=-2.6903, interpMode=RCIM_Cubic, leaveTangent=-2.6903, tangentMode=RCTM_User, time=-0.0027, value=0}
           {interpMode=RCIM_Cubic, tangentMode=RCTM_User, time=0.3, value=-1}
           {interpMode=RCIM_Cubic, leaveTangent=1.4239, time=0.7, value=1}
           <at 1 0 RCIM_Cubic>
          ]}}}
       CurveIndex=$Particles.RandomPosition
       Scale Curve=0
      )
      B=&FloatFromCurve(
       FloatCurve=DataInterface() {propertyValues {Curve {keys [
           {arriveTangent=-0.8464, interpMode=RCIM_Cubic, time=0, value=0.25}
           {arriveTangent=1.1407, interpMode=RCIM_Cubic, leaveTangent=1.1407, tangentMode=RCTM_User, time=1, value=1}
          ]}}}
       CurveIndex=$Particles.RandomPosition
      )
     )
     Z=&FloatFromCurve(
      FloatCurve=DataInterface() {propertyValues {Curve {keys [
          {arriveTangent=7.2867, interpMode=RCIM_Cubic, time=0, value=1}
          {arriveTangent=-4.3515, interpMode=RCIM_Cubic, leaveTangent=-4.3515, tangentMode=RCTM_User, time=0.2171, value=0.5268}
          {arriveTangent=-0.0104, interpMode=RCIM_Cubic, leaveTangent=-0.0104, tangentMode=RCTM_User, time=0.45, value=0}
          {arriveTangent=4.8334, interpMode=RCIM_Cubic, leaveTangent=4.8334, tangentMode=RCTM_User, time=0.8116, value=0.579}
          {arriveTangent=-2.1665, interpMode=RCIM_Cubic, leaveTangent=-2.1665, tangentMode=RCTM_User, time=0.9827, value=0.7}
         ]}}}
      CurveIndex=$Particles.RandomPosition
      Scale Curve=900
     )
    )) {isSetParameters=true}
   &SphereLocation(
    MaskSpawnGroup=false
    Offset=<0 0 500>
    Override Seed=false
    Sphere Radius=&Multiply_Float(A=&Lerp_Float(A=100, B=200, Alpha=0.5), B=7)
    Surface Only=false
    UseNonUniformScale=false
    UseOffset=true
    UseSphereOrientation=false
   ) {module=SphereLocation001}
   /SetVariables(Particles.Sparkle:NiagaraFloat=&UniformAOrBFloat(A=0, Distribution Weight=0.2, Evaluation Type=::NewEnumerator0, Recalculate Random Each Loop=true, Override Seed=false)) {isSetParameters=true}
   &AddVelocity(
    Scale Added Velocity=&VectorFromFloat(Value=0)
    Velocity=&RotateVectorByQuaternion(
     Vector To Rotate=&NormalizeVector(Vector To Normalize=&Subtract_Vector(A=Linked() {linkedVariable {name=Particles.Position, type=NiagaraPosition}}, B=<1200 0 500>))
     Quaternion=<0 -1 0 1>
    )
   )
  ]
  particleUpdateScript [
   &ParticleState()
   /SetVariables(Particles.HeightLerp:NiagaraFloat=&FloatFromCurve(
     FloatCurve=DataInterface() {propertyValues {Curve {keys [{interpMode=RCIM_Cubic, tangentMode=RCTM_Break, time=0, value=1}, {interpMode=RCIM_Cubic, tangentMode=RCTM_Break, time=200, value=0}]}}}
     CurveIndex=&Subtract_Float(
      A=$System.Height
      B=&MakeFloatFromVector(Vector=Linked() {linkedVariable {name=Particles.Position, type=NiagaraPosition}}, Channel=::NewEnumerator2)
     )
    )) {isSetParameters=true, module=SetVariables_#2}
   &DynamicMaterialParameters(
    Index 0 Param 1=1
    Index 0 Param 2=1
    Index 0 Param 3=0.5
    Module.Param0WriteEnabled=true
    Module.Param1WriteEnabled=true
    Module.Param2WriteEnabled=true
    Module.Param3WriteEnabled=true
   )
   &ScaleColor(
    Scale Alpha=&FloatFromCurve(
     FloatCurve=DataInterface() {propertyValues {Curve {keys [
         <at 0 0 RCIM_Cubic>
         {interpMode=RCIM_Cubic, tangentMode=RCTM_User, time=0.1, value=0.9914}
         {interpMode=RCIM_Cubic, tangentMode=RCTM_User, time=0.7, value=1}
         <at 1 0>
        ]}}}
     Scale Curve=&Lerp_Float(A=0.25, B=1, Alpha=$System.DistanceFade)
    )
    Scale RGB=&MakeVectorFromLinearColorRGB(Color=&Multiply_LinearColorByFloat(
      Linear Color=&LerpLinearColors(StartColor=<rgba 0.936 0.101 1 1>, EndColor=<rgba 0.2 0.3 1 1>, LerpFactor=$Particles.HeightLerp)
      Float=$User.TreeEnergy_Brightness
     ))
    ScaleA=true
    ScaleRGB=true
   )
   &ScaleSpriteSize(Scale Factor=&Vector2DFromFloat(Value=&FloatFromCurve(
      FloatCurve=DataInterface() {propertyValues {Curve {keys [
          <at 0 0 RCIM_Cubic>
          {arriveTangent=2.4312, interpMode=RCIM_Cubic, leaveTangent=2.4312, time=0.1483, value=1}
          {arriveTangent=-0.0468, interpMode=RCIM_Cubic, leaveTangent=-0.0468, time=0.2916, value=0.709}
          {arriveTangent=-0.2796, interpMode=RCIM_Cubic, leaveTangent=-0.2796, time=0.4899, value=0.984}
          {arriveTangent=0.7494, interpMode=RCIM_Cubic, leaveTangent=0.7494, tangentMode=RCTM_User, time=0.6851, value=0.599}
          <at 1 0>
         ]}}}
      Scale Curve=&Lerp_Float(A=0.5, B=5, Alpha=0.5)
     )))
   &VortexForce(
    Influence Falloff Radius=1600
    Origin Pull Amount=-0.5
    Use Influence Falloff=true
    Vortex Axis=<0 -1 0>
    Vortex Force Amount=100
    Vortex Origin:Vector3f=<1500 0 400>
   )
   &VectorNoiseForce(Force Amount=100, Module.Use Random Noise Time Interval=true)
   /SetVariables(Particles.SparkleSize:Vector2f=/Niagara/DynamicInputs/Add/Add_Vector2(
     A:Vector2f=$Particles.SpriteSize
     B:Vector2f=&Vector2DFromFloat(Value=&Multiply_Float(
       A=$Particles.Sparkle
       B=&FloatFromCurve(
        FloatCurve=DataInterface() {propertyValues {Curve {keys [
            <at 0.4 0.05 RCIM_Cubic>
            {arriveTangent=-4.4497, interpMode=RCIM_Cubic, leaveTangent=-4.4497, tangentMode=RCTM_User, time=0.45, value=1}
            <at 0.6 0.05 RCIM_Cubic>
           ]}}}
        Scale Curve=700
       )
      ))
    )) {isSetParameters=true}
   &Drag(Use Linear Drag=true, Use Rotational Drag=true)
   &GravityForce(Gravity=<0 0 50>)
   &SolveForcesAndVelocity(Clamp Velocity=false, Limit Acceleration=false)
  ]
  rendererClasses [&NiagaraSpriteRendererProperties]
  renderers [
   {rendererClass=&NiagaraSpriteRendererProperties}
   {rendererClass=&NiagaraSpriteRendererProperties, rendererIndex=1}
  ]
 }
 Shimmers_Top {
  simTarget=GPUComputeSim
  emitterUpdateScript [
   &EmitterState()
   &SpawnRate(SpawnRate=&Lerp_Float(A=0, B=30, Alpha=$User.Fade out), Use Spawn Probability=true)
  ]
  particleSpawnScript [
   &InitializeParticle(
    Lifetime Mode=::NewEnumerator1
    Position Mode=::NewEnumerator1
    Sprite Rotation Mode=::NewEnumerator2
    Sprite Size=&RandomRangeVector2D(Minimum=<400 400>, Maximum=<800 800>, Evaluation Type=::NewEnumerator1)
    Sprite Size Mode=::NewEnumerator4
    UsePositionOffset=false
    Write Material Random=true
   )
   &AddVelocity(Velocity=&RandomRangeVector(Minimum=<-100 -100 -100>, Maximum=<100 100 100>))
   &DynamicMaterialParameters(
    Index 0 Param 1=&RandomRangeFloat(Minimum=0, Maximum=1, Evaluation Type=::NewEnumerator1)
    Index 0 Param 2=&RandomRangeFloat(Minimum=0.4, Maximum=0.6, Evaluation Type=::NewEnumerator1)
    Index 0 Param 3=0.2
    Module.Index0Float0Write=true
    Module.Index0Float1Write=true
    Module.Index0Float2Write=true
    Module.Index0Float3Write=true
   )
   &ShapeLocation(
    MaskSpawnGroup=false
    Offset=<0 0 1400>
    Offset Mode=::NewEnumerator0
    Override Seed=false
    Sphere Radius=790
    Sphere Surface Distribution=0.8
    UseOwnerScale=true
   )
  ]
  particleUpdateScript [
   &ParticleState()
   /SetVariables(Particles.HeightLerp:NiagaraFloat=&FloatFromCurve(
     FloatCurve=DataInterface() {propertyValues {Curve {keys [{interpMode=RCIM_Cubic, tangentMode=RCTM_Break, time=0, value=1}, {interpMode=RCIM_Cubic, tangentMode=RCTM_Break, time=200, value=0}]}}}
     CurveIndex=&Subtract_Float(
      A=$System.Height
      B=&MakeFloatFromVector(Vector=Linked() {linkedVariable {name=Particles.Position, type=NiagaraPosition}}, Channel=::NewEnumerator2)
     )
    )) {isSetParameters=true, module=SetVariables_#2}
   &Drag(Use Linear Drag=true, Use Rotational Drag=true)
   &SolveForcesAndVelocity(Clamp Velocity=false, Limit Acceleration=false)
   &SpriteRotationRate(Rotation Rate=&RandomRangeFloat(Minimum=-20, Maximum=20))
   &ScaleColor(
    Scale Alpha=&FloatFromCurve(
     FloatCurve=DataInterface() {propertyValues {Curve {keys [{arriveTangent=1.0833, interpMode=RCIM_Cubic, leaveTangent=1.0833, time=0, value=0}, <at 0.6 0.65 RCIM_Cubic>, <at 1 0>]}}}
     Scale Curve=&RandomRangeFloat(Minimum=0.5, Maximum=1)
    )
    Scale RGB=&MakeVectorFromLinearColorRGB(Color=&Multiply_LinearColorByFloat(
      Linear Color=&LerpLinearColors(StartColor=<rgba 0.936 0.101 1 1>, EndColor=<rgba 0.2 0.3 1 1>, LerpFactor=$Particles.HeightLerp)
      Float=$User.TreeEnergy_Brightness
     ))
    ScaleA=true
    ScaleRGB=true
   )
   &ScaleSpriteSize(
    Uniform Curve Scale=2
    Uniform Curve Sprite Scale=DataInterface() {propertyValues {Curve {keys [<at 0 0.7 RCIM_Cubic>, {interpMode=RCIM_Cubic, tangentMode=RCTM_Break, time=1, value=1}]}}}
   )
   &CameraOffset(Camera Offset Amount=&RandomRangeFloat(Minimum=-150, Maximum=200))
  ]
  rendererClasses [&NiagaraSpriteRendererProperties]
  renderers [{rendererClass=&NiagaraSpriteRendererProperties}]
 }
 Shimmers_Tree {
  emitterUpdateScript [
   &EmitterState()
   &SpawnRate(SpawnRate=&Lerp_Float(A=0, B=40, Alpha=$User.Fade out), Use Spawn Probability=true)
  ]
  particleSpawnScript [
   &InitializeParticle(
    Lifetime Mode=::NewEnumerator1
    Position Mode=::NewEnumerator1
    Sprite Rotation Mode=::NewEnumerator2
    Sprite Size=&RandomRangeVector2D(Minimum=<400 400>, Maximum=<800 800>, Evaluation Type=::NewEnumerator1)
    Sprite Size Mode=::NewEnumerator4
    UsePositionOffset=false
    Write Material Random=true
   )
   /Niagara/Modules/Spawn/Location/StaticMeshLocation(
    MaskSpawnGroup=false
    OffsetAlongNormal=false
    OverrideRandomSeed=false
    SamplePosOffset=false
    Static Mesh=DataInterface() {propertyValues {
      DefaultMesh=/LimeRock_Tree/Assets/Treehouse/SM_LimeRock_Treehouse_Placeholder
      InstanceIndex=-1
      LODIndex=0
      PreviewMesh=&SM_LimeRock_Treehouse_Placeholder
      SoftSourceActor=null
      SourceMode=Default
      bAllowSamplingFromStreamingLODs=false
      bCaptureTransformsPerFrame=true
      bUsePhysicsBodyVelocity=false
      FilteredSockets []
      LODIndexUserParameter {parameter {name=None, typeDefHandle {registeredTypeIndex=108}, varData []}}
      MeshParameterBinding {parameter {name=None, typeDefHandle {registeredTypeIndex=1}, varData []}}
      SectionFilter {allowedMaterialSlots []}
     }}
   )
   &AddVelocity(Velocity=&RandomRangeVector(Minimum=<-100 -100 -100>, Maximum=<100 100 200>))
   &DynamicMaterialParameters(
    Index 0 Param 1=&RandomRangeFloat(Minimum=0, Maximum=1, Evaluation Type=::NewEnumerator1)
    Index 0 Param 2=&RandomRangeFloat(Minimum=0.4, Maximum=0.6, Evaluation Type=::NewEnumerator1)
    Index 0 Param 3=0.2
    Module.Param0WriteEnabled=true
    Module.Param1WriteEnabled=true
    Module.Param2WriteEnabled=true
    Module.Param3WriteEnabled=true
   )
  ]
  particleUpdateScript [
   &ParticleState()
   /SetVariables(Particles.HeightLerp:NiagaraFloat=&FloatFromCurve(
     FloatCurve=DataInterface() {propertyValues {Curve {keys [{interpMode=RCIM_Cubic, tangentMode=RCTM_Break, time=0, value=1}, {interpMode=RCIM_Cubic, tangentMode=RCTM_Break, time=200, value=0}]}}}
     CurveIndex=&Subtract_Float(
      A=$System.Height
      B=&MakeFloatFromVector(Vector=Linked() {linkedVariable {name=Particles.Position, type=NiagaraPosition}}, Channel=::NewEnumerator2)
     )
    )) {isSetParameters=true, module=SetVariables_#2}
   &SolveForcesAndVelocity(Clamp Velocity=false, Limit Acceleration=false)
   &SpriteRotationRate(Rotation Rate=&RandomRangeFloat(Minimum=-20, Maximum=20))
   &ScaleColor(
    Scale Alpha=&FloatFromCurve(
     FloatCurve=DataInterface() {propertyValues {Curve {keys [{arriveTangent=1.0833, interpMode=RCIM_Cubic, leaveTangent=1.0833, time=0, value=0}, <at 0.6 0.65 RCIM_Cubic>, <at 1 0>]}}}
     Scale Curve=&RandomRangeFloat(Minimum=0.5, Maximum=1)
    )
    Scale RGB=&MakeVectorFromLinearColorRGB(Color=&Multiply_LinearColorByFloat(
      Linear Color=&LerpLinearColors(StartColor=<rgba 0.936 0.101 1 1>, EndColor=<rgba 0.2 0.3 1 1>, LerpFactor=$Particles.HeightLerp)
      Float=$User.TreeEnergy_Brightness
     ))
    ScaleA=true
    ScaleRGB=true
   )
   &ScaleSpriteSize(
    Uniform Curve Scale=2
    Uniform Curve Sprite Scale=DataInterface() {propertyValues {Curve {keys [<at 0 0.7 RCIM_Cubic>, {interpMode=RCIM_Cubic, tangentMode=RCTM_Break, time=1, value=1}]}}}
   )
   &CameraOffset(Camera Offset Amount=&RandomRangeFloat(Minimum=-150, Maximum=200))
  ]
  rendererClasses [&NiagaraSpriteRendererProperties]
  renderers [{rendererClass=&NiagaraSpriteRendererProperties}]
 }
 SmallWave {
  emitterUpdateScript [
   &EmitterState()
   &SpawnRate(SpawnRate=&Multiply_Float(A=2, B=$User.Energy_SpawnRate), Use Spawn Probability=false)
  ]
  particleSpawnScript [
   &InitializeParticle(
    Color Mode=::NewEnumerator0
    Lifetime=$User.Energy_SpawnTimeMult
    Mesh Scale=<5 7 7>
    Mesh Scale Mode=::NewEnumerator4
    Position=<0 0 1000>
    Position Mode=::NewEnumerator1
    UsePositionOffset=false
    Write Material Random=true
   ) {module=InitializeParticle001}
   &CameraOffset(Camera Offset Amount=100)
  ]
  particleUpdateScript [
   &ParticleState()
   &ScaleMeshSize(Scale Factor=&VectorFromFloat(Value=&FloatFromCurve(
      FloatCurve=DataInterface() {propertyValues {Curve {
         defaultValue=1
         keys [
          {arriveTangent=-0.0458, interpMode=RCIM_Cubic, leaveTangent=-0.0458, tangentMode=RCTM_User, time=0, value=1}
          {arriveTangent=-3.2857, interpMode=RCIM_Cubic, leaveTangent=-3.2857, tangentMode=RCTM_User, time=0.7, value=0}
         ]
        }}}
      Scale Curve=&Multiply_Float(A=0.25, B=$User.Uniform Scale)
     ))) {module=ScaleMeshSize001}
   &ScaleColor(
    Scale Alpha=&FloatFromCurve(
     FloatCurve=DataInterface() {propertyValues {Curve {keys [
         {arriveTangent=4.499, interpMode=RCIM_Cubic, leaveTangent=4.499, leaveTangentWeight=0.7975, tangentMode=RCTM_User, time=0, value=0}
         {arriveTangentWeight=0.4526, interpMode=RCIM_Cubic, leaveTangentWeight=0.4526, tangentMode=RCTM_User, time=0.3, value=1}
         {arriveTangent=-0.0962, interpMode=RCIM_Cubic, leaveTangent=-0.0962, tangentMode=RCTM_User, time=1, value=0}
        ]}}}
     Scale Curve=0.15
    )
    Scale RGB=&VectorFromFloat(Value=&FloatFromCurve(
      FloatCurve=DataInterface() {propertyValues {Curve {
         defaultValue=0
         keys [
          {arriveTangent=4.1785, interpMode=RCIM_Cubic, leaveTangent=4.1785, tangentMode=RCTM_User, time=0, value=0}
          {arriveTangent=1.1179, interpMode=RCIM_Cubic, leaveTangent=1.1179, tangentMode=RCTM_User, time=0.2042, value=0.6228}
          {arriveTangent=0.0224, interpMode=RCIM_Cubic, leaveTangent=0.0224, tangentMode=RCTM_User, time=1, value=1}
         ]
        }}}
      Scale Curve=&FNTimeOfDayScalar(Night Scale=0.2, Morning Scale=0.3, Evening Scale=0.3)
     ))
    ScaleA=true
    ScaleRGB=true
   ) {module=ScaleColor001}
   &DynamicMaterialParameters(
    Index 0 Param 1=&FloatFromCurve(FloatCurve=DataInterface() {propertyValues {Curve {
        defaultValue=0
        keys [
         {arriveTangent=4.1785, interpMode=RCIM_Cubic, leaveTangent=4.1785, tangentMode=RCTM_User, time=0, value=0}
         {arriveTangent=1.1179, interpMode=RCIM_Cubic, leaveTangent=1.1179, tangentMode=RCTM_User, time=0.2042, value=0.6228}
         {arriveTangent=0.0224, interpMode=RCIM_Cubic, leaveTangent=0.0224, tangentMode=RCTM_User, time=1, value=1}
        ]
       }}})
    Index 0 Param 2=&FloatFromCurve(FloatCurve=DataInterface() {propertyValues {Curve {
        defaultValue=1
        keys [{interpMode=RCIM_Cubic, tangentMode=RCTM_User, time=0, value=-0.7}, {interpMode=RCIM_Cubic, tangentMode=RCTM_User, time=1, value=0}]
       }}})
    Index 0 Param 3=&FloatFromCurve(
     FloatCurve=DataInterface() {propertyValues {Curve {
        defaultValue=0
        keys [{interpMode=RCIM_Cubic, tangentMode=RCTM_User, time=0, value=1}, {interpMode=RCIM_Cubic, tangentMode=RCTM_User, time=1, value=-1.5217}]
       }}}
     Scale Curve=0.3
    )
    Index 0 Param 4=&RandomRangeFloat(Minimum=1, Maximum=1)
    Index 1 Param 1=&FloatFromCurve(FloatCurve=DataInterface() {propertyValues {Curve {keys [
         {arriveTangent=-0.0851, interpMode=RCIM_Cubic, leaveTangent=-0.0851, tangentMode=RCTM_User, time=0, value=0}
         {arriveTangent=1.1179, leaveTangent=1.1179, time=0.2277, value=1}
         {arriveTangent=0.0224, interpMode=RCIM_Cubic, leaveTangent=0.0224, tangentMode=RCTM_User, time=1, value=1}
        ]}}})
    Index 2 Param 1=&FloatFromCurve(FloatCurve=DataInterface() {propertyValues {Curve {keys [
         {
          arriveTangent=-0.0147
          arriveTangentWeight=0.7781
          interpMode=RCIM_Cubic
          leaveTangent=-0.0147
          leaveTangentWeight=0.7781
          tangentMode=RCTM_User
          tangentWeightMode=RCTWM_WeightedBoth
          time=0
          value=0.2
         }
         {arriveTangent=-0.3563, interpMode=RCIM_Cubic, leaveTangent=-0.3563, tangentMode=RCTM_User, time=1, value=2}
        ]}}})
    Module.Index0Float0Write=true
    Module.Index0Float1Write=true
    Module.Index0Float2Write=true
    Module.Index0Float3Write=true
    Module.Index1Float0Write=true
    Module.Index1Float1Write=true
    Module.Index1Float2Write=true
    Module.Index1Float3Write=true
    Module.Index2Float0Write=true
    Module.Index2Float1Write=true
    Module.Index2Float2Write=true
    Module.Index2Float3Write=true
    Write Parameter Index 1=::NewEnumerator1
    Write Parameter Index 2=::NewEnumerator1
   )
   &SolveForcesAndVelocity(Clamp Velocity=false, Limit Acceleration=false)
  ]
  rendererClasses [&NiagaraMeshRendererProperties]
  renderers [{rendererClass=&NiagaraMeshRendererProperties, rendererIndex=0}]
 }
 Tree_Decal_Energy001 {
  emitterUpdateScript [
   &EmitterState(Life Cycle Mode=::NewEnumerator1, Loop Duration=5.4, UseLoopDelay=false)
   &SpawnBurst_Instantaneous(Spawn Count=$User.TreeHouse_TopEnergy_MeshDecal_Spawn, Use Spawn Probability=false, UseLoopCountLimit=false)
  ]
  particleSpawnScript [&InitializeParticle(
    Lifetime=5.5
    Mesh Scale Mode=::NewEnumerator3
    Mesh Uniform Scale=40
    Position Offset=<0 0 3200>
    UsePositionOffset=true
    Write Material Random=true
   )]
  particleUpdateScript [
   &ParticleState()
   &DynamicMaterialParameters(
    Index 0 Param 1=1
    Index 0 Param 2=&Lerp_Float(
     A=1
     B=0.3
     Alpha=&FloatFromCurve(FloatCurve=DataInterface() {propertyValues {Curve {keys [
          {arriveTangent=5.1433, interpMode=RCIM_Cubic, time=0, value=0}
          {arriveTangent=4.5726, interpMode=RCIM_Cubic, leaveTangent=4.5726, tangentMode=RCTM_User, time=0.25, value=0.7}
          {arriveTangent=-6.4921, interpMode=RCIM_Cubic, leaveTangent=2.9849, tangentMode=RCTM_Break, time=0.4, value=0.35}
          {arriveTangent=0.6574, interpMode=RCIM_Cubic, leaveTangent=0.6574, tangentMode=RCTM_User, time=1, value=1}
         ]}}})
    )
    Index 3 Param 1=&Multiply_Float(A=&MakeFloatFromVector(Vector=$Particles.Scale, Channel=::NewEnumerator0), B=90)
    Index 3 Param 2=&FloatFromCurve(
     FloatCurve=DataInterface() {propertyValues {Curve {keys [
         {arriveTangent=2.1437, interpMode=RCIM_Cubic, leaveTangent=2.1437, tangentMode=RCTM_User, time=0, value=0}
         {arriveTangent=1.3259, interpMode=RCIM_Cubic, leaveTangent=1.3259, tangentMode=RCTM_User, time=0.6, value=0.5}
         {interpMode=RCIM_Cubic, leaveTangent=0.2859, time=1, value=1}
        ]}}}
     Scale Curve=0
    )
    Module.Param0WriteEnabled=true
    Module.Param0WriteEnabled3=true
    Module.Param1WriteEnabled=true
    Module.Param1WriteEnabled3=true
    Module.Param2WriteEnabled=false
    Module.Param2WriteEnabled3=false
    Module.Param3WriteEnabled=false
    Module.Param3WriteEnabled3=false
    Write Parameter Index 3=true
   )
   /SetVariables(Particles.WaveColor:LinearColor=&MakeLinearColorFromVectorAndFloat(
     "Vector (RGB)"=&Multiply_Vector(
      A=&MakeVectorFromLinearColorRGB(Color=$User.OverallColorSwitch)
      B=&Lerp_Vector(
       A=&MakeVectorFromLinearColorRGB(Color=<rgba 0.4843 0.1833 0.5677 1>)
       B=&MakeVectorFromLinearColorRGB(Color=<rgba 0.3417 0.2639 0.7135 1>)
       Alpha=&FloatFromCurve(FloatCurve=DataInterface() {propertyValues {Curve {
           defaultValue=1
           keys [
            {interpMode=RCIM_Cubic, tangentMode=RCTM_User, time=0, value=0}
            {arriveTangent=-0.0252, interpMode=RCIM_Cubic, leaveTangent=-0.0252, tangentMode=RCTM_User, time=0.0954, value=0.8302}
            {interpMode=RCIM_Cubic, tangentMode=RCTM_User, time=0.8483, value=0.8627}
            {tangentMode=RCTM_User, time=1, value=0}
           ]
          }}})
      )
     )
     "Float (Alpha)"=&FloatFromCurve(
      FloatCurve=DataInterface() {propertyValues {Curve {
         defaultValue=1
         keys [
          {interpMode=RCIM_Cubic, tangentMode=RCTM_User, time=0, value=0}
          {interpMode=RCIM_Cubic, tangentMode=RCTM_User, time=0.0967, value=0.9353}
          {arriveTangent=-0.1559, interpMode=RCIM_Cubic, leaveTangent=-0.1559, tangentMode=RCTM_User, time=0.9048, value=0.8647}
          {tangentMode=RCTM_User, time=1, value=0}
         ]
        }}}
      Scale Curve=&Lerp_Float(A=0, B=1, Alpha=$User.Fade out)
     )
    )) {isSetParameters=true}
  ]
  rendererClasses [&NiagaraMeshRendererProperties]
  renderers [{rendererClass=&NiagaraMeshRendererProperties, rendererIndex=0}]
 }
 Treehouse_Splines {
  emitterUpdateScript [
   &EmitterState(Life Cycle Mode=::NewEnumerator1, Loop Behavior=::NewEnumerator1, Loop Duration=3, UseLoopDelay=false)
   &SpawnBurst_Instantaneous(Use Spawn Probability=false, UseLoopCountLimit=false)
  ]
  particleSpawnScript [&InitializeParticle(
    Lifetime=3
    Mesh Scale Mode=::NewEnumerator3
    Position=<0 0 70>
    Position Mode=::NewEnumerator1
    UsePositionOffset=false
    Write Material Random=true
   )]
  particleUpdateScript [
   &ParticleState(Kill Particles When Lifetime Has Elapsed=false)
   &DynamicMaterialParameters(
    Index 0 Param 1=1.2
    Index 0 Param 2=0.7
    Index 0 Param 3=$User.TreeHouse_Spline_ColorLerp
    Index 0 Param 4=$User.TreeHouse_Spline_VerticalOpacityPower
    Module.Index0Float0Write=true
    Module.Index0Float1Write=true
    Module.Index0Float2Write=true
    Module.Index0Float3Write=true
   )
   &ScaleColor(
    Scale Alpha=&Multiply_Float(A=$User.Fade out, B=0.5)
    Scale RGB=&VectorFromFloat(Value=5)
    ScaleA=true
    ScaleRGB=true
   )
  ]
  rendererClasses [&NiagaraMeshRendererProperties]
  renderers [{rendererClass=&NiagaraMeshRendererProperties, rendererIndex=0}]
 }
}
summary {
 systemName=NS_LimeRock_Treehouse_FX_2
 emitters [
  {emitterName=Motes, rendererClasses [&NiagaraSpriteRendererProperties]}
  {emitterName=Shimmers_Tree, rendererClasses [&NiagaraSpriteRendererProperties]}
  {emitterName=LightningSmall_Top, rendererClasses [&NiagaraMeshRendererProperties]}
  {emitterName=Light001, rendererClasses [&NiagaraLightRendererProperties]}
  {emitterName=Treehouse_Splines, rendererClasses [&NiagaraMeshRendererProperties]}
  {emitterName=Shimmers_Top, simTarget=GPUComputeSim, rendererClasses [&NiagaraSpriteRendererProperties]}
  {emitterName=LightningSmall_Trunk, rendererClasses [&NiagaraMeshRendererProperties]}
  {emitterName=BigWave, rendererClasses [&NiagaraMeshRendererProperties]}
  {emitterName=SmallWave, rendererClasses [&NiagaraMeshRendererProperties]}
  {emitterName=Decal_Energy_Ground_Right001, rendererClasses [&NiagaraDecalRendererProperties]}
  {emitterName=Tree_Decal_Energy001, rendererClasses [&NiagaraMeshRendererProperties]}
 ]
 userVariables [
  User.ColorShifting {type=NiagaraFloat, defaultValue=0, description=""}
  User.Energy_SpawnRate {type=NiagaraFloat, defaultValue=0, description=""}
  User.Energy_SpawnTimeMult {type=NiagaraFloat, defaultValue=3, description=""}
  User.Fade out {type=NiagaraFloat, defaultValue=0.3, description=""}
  User.GroundDecal_Intensity {type=NiagaraFloat, defaultValue=0.2, description=""}
  User.Light_Intensity {type=NiagaraFloat, defaultValue=0.3, description=""}
  User.Light_Radius {type=NiagaraFloat, defaultValue=1, description=""}
  User.LightColor_Switch {type=NiagaraFloat, defaultValue=0, description=""}
  User.Lightning Spawn Rate {type=NiagaraFloat, defaultValue=2, description=""}
  User.Opacity {type=NiagaraFloat, defaultValue=1, description=""}
  User.OverallColorSwitch {type=LinearColor, description="", defaultValue {struct=LinearColor, value=<rgba 0.9357 0.101 1 1>}}
  User.TreeEnergy_Brightness {type=NiagaraFloat, defaultValue=0.3, description=""}
  User.TreeHouse_EnergyLine_V_Offset {type=NiagaraFloat, defaultValue=-0.46, description=""}
  User.TreeHouse_Spline_ColorLerp {type=NiagaraFloat, defaultValue=0, description=""}
  User.TreeHouse_Spline_VerticalOpacityPower {type=NiagaraFloat, defaultValue=3, description=""}
  User.TreeHouse_TopEnergy_MeshDecal_Spawn {type=NiagaraInt32}
  User.TreePosition {type=NiagaraPosition, description="", defaultValue {}}
  User.TreeTrunk_LightningRate {type=NiagaraFloat, defaultValue=1, description=""}
  User.Uniform Scale {type=NiagaraFloat, defaultValue=2, description=""}
 ]
}
